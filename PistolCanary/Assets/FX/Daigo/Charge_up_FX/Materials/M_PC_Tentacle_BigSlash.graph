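# Unlit, two-sided translucent material for a big tentacle slash effect on Niagara mesh particles and ribbons.
# Opacity combines a noise-distorted slash texture, sine UV edge fades, optional tapering, smoke-based HeightLerp dissolve, view-angle fade and depth fade.
# Emissive multiplies two curve-atlas gradients along V and U by particle color and panning flame noise, then applies desaturation and emissive adjustments.
# Dynamic parameters drive dissolve, mask pan time and taper direction.
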Material /PistolCanary/Assets/FX/Daigo/Charge_up_FX/Materials/M_PC_Tentacle_BigSlash
graph {
 properties {
  bAutomaticallySetUsageInEditor=false
  bUsedWithNiagaraMeshParticles=true
  bUsedWithNiagaraRibbons=true
  blendMode=BLEND_Translucent
  shadingModel=MSM_Unlit
  twoSided=true
  usedShadingModels=Unlit
 }
 nodes [
  n1 = MaterialFunctionCall() {materialFunction=&LinearGradient}
  n2 = OneMinus(None=@n1.UGradient)
  n3 = DynamicParameter() {defaultValue=<rgba 0 0.2 0 0>, paramNames [Dissolve, DissolveOffset, ColorNoiseOffset, Param4]}
  n4 = LinearInterpolate(
   B=Multiply(A=LinearInterpolate(A=@n1.UGradient, B=@n2, Alpha=@n3.Param4), B=@n1.VGradient)
   Alpha=ScalarParameter() {parameterName=UseTapepedShape}
  ) {constA=1}
  n5 = VectorParameter() {parameterName=UVScale, channelNames {a="", b="", g=V Scale, r=U Scale}, defaultValue=<rgba 0.3 1 0 0>}
  n6 = VectorParameter() {parameterName=UVOffset, channelNames {a="", b="", g=V Offset, r=U Offset}, defaultValue=<rgba 0.2 0 0 0>}
  n7 = Add(A=Multiply(A=TextureCoordinate(), B=ComponentMask(None=@n5) {g=true, r=true}), B=ComponentMask(None=@n6) {g=true, r=true})
  n8 = TextureCoordinate() {vTiling=0.7}
  n9 = Multiply(
   A=TextureSample(UVs=Panner(Coordinate=@n8) {speedX=1}) {samplerType=SAMPLERTYPE_Grayscale, texture=&LowResBlurredNoise}.R
   B=ScalarParameter() {parameterName=Disturb}
  ) {constB=0.2}
  n10 = VectorParameter() {parameterName=MaskPanner}
  n11 = TextureSampleParameter2D(UVs=Panner(
    Coordinate=Add(A=@n7, B=@n9)
    Time=DynamicParameter() {defaultValue=<rgba 0 0.2 0 0>, paramNames [Dissolve, DissolveOffset, ColorNoiseOffset, Param4]}.ColorNoiseOffset
    Speed=ComponentMask(None=@n10.RGB) {g=true, r=true}
   )) {
   parameterName=Defalut Tex
   samplerSource=SSM_Clamp_WorldGroupSettings
   texture=&/T_PC_Tentacle_Slash
  }
  n12 = TextureCoordinate()
  n13 = Multiply(A=StaticComponentMaskParameter(None=@n11.RGB) {defaultB=true, parameterName=Mask}, B=Sine(None=ComponentMask(None=@n12) {r=true}) {period=2})
  n14 = DynamicParameter() {defaultValue=<rgba 0 0.2 0 0>, paramNames [Dissolve, DissolveOffset, ColorNoiseOffset, Param4]}
  n15 = TextureSample(UVs=TextureCoordinate() {uTiling=2, vTiling=3}) {texture=&T_Inky_Smoke_Tile}
  n16 = Saturate(None=Multiply(A=MaterialFunctionCall(Transition Phase=@n14.Dissolve, Height Texture=@n15.R) {materialFunction=&HeightLerp}.Results))
  n17 = Multiply(A=LinearInterpolate(Alpha=VertexColor().R) {constB=5}, B=ScalarParameter() {defaultValue=1, parameterName="FadeSensitivity (optional) (S)"})
  n18 = TextureCoordinate()
  n19 = Multiply(
   A=Multiply(
    A=Multiply(A=@n13, B=@n16)
    B=MaterialFunctionCall("FadeSensitivity (optional)"=@n17) {materialFunction=&FadeBasedOnViewAngle}.Result
   )
   B=Sine(None=ComponentMask(None=@n18) {g=true}) {period=2}
  )
  DepthFade(
   Opacity=Multiply(A=Multiply(A=@n4, B=@n19), B=ParticleColor().A)
   FadeDistance=ScalarParameter() {defaultValue=50, parameterName=FadeDistance}
  ) {fadeDistanceDefault=50}
  MaterialFunctionCall() {materialFunction=&DebugTimeSine}
  n20 = TextureCoordinate()
  n21 = CurveAtlasRowParameter(CurveTime=ComponentMask(None=@n20) {g=true}) {
   atlas=/PistolCanary/Assets/FX/Daigo/Charge_up_FX/CurveAtlas/CAtlas_Masamune_Color
   curve=/PistolCanary/Assets/FX/Daigo/Charge_up_FX/CurveAtlas/CV_Masamune_SlashColor
   parameterName=Color
  }
  n22 = CurveAtlasRowParameter(CurveTime=ComponentMask(None=@n20) {r=true}) {
   atlas=&CAtlas_Masamune_Color
   curve=/PistolCanary/Assets/FX/Daigo/Charge_up_FX/CurveAtlas/CV_Masamune_SlashColor2
   defaultValue=2
   parameterName=Color2
  }
  n23 = Multiply(A=@n21, B=@n22)
  n24 = ScalarParameter() {defaultValue=1, parameterName=MainColorIntensity}
  n25 = ScalarParameter() {defaultValue=2, parameterName=Detail Noise Max Intensity}
  n26 = TextureCoordinate() {uTiling=0.5}
  n27 = Power(Base=TextureSample(UVs=Panner(Coordinate=@n26) {speedX=-0.5}) {samplerType=SAMPLERTYPE_LinearGrayscale, texture=/Game/Weapons/FORT_Melee/Pickaxe_Cube_Ninja_Male/FX/T_FluidFlameTile}.R) {constExponent=1}
  n28 = Multiply(
   A=Multiply(A=ParticleColor().RGB, B=Multiply(A=@n23, B=@n24))
   B=LinearInterpolate(B=LinearInterpolate(B=@n25, Alpha=@n27) {constA=0.7, constB=2}, Alpha=ScalarParameter() {defaultValue=1, parameterName=UseDetail Noise}) {constA=1}
  )
  n29 = ScalarParameter() {parameterName=Des}
  Reroute(None=MaterialFunctionCall(Emissive=Desaturation(None=@n28, Fraction=@n29)) {materialFunction=&MF_FX_EmissiveAdjustments}.Result)
 ]
}
