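# Niagara system for the Pistol Canary drain-energy beam, drawing muzzle flares and energy ribbons from the local origin to User.Beam End.
# Two ribbon emitters partition particles along that line with an arced offset; one adds waveform wobble.
# User.EnergyBeamSwipe drives wipe masks on the wobbling ribbons, feeds the other ribbon's material, and cuts off the flares past halfway.
# Flare spawn rate follows platform quality; System.FadeOut scales ribbon width.
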NiagaraSystem /PistolCanary/Assets/FX/Daigo/Charge_up_FX/NS_PistolCanary_MM_DrainEnergy_Beam
dependencies {
 usedDataInterfaces [
  &NiagaraDataInterfaceCurve
  &NiagaraDataInterfaceColorCurve
  &NiagaraDataInterfacePlatformSet
  &NiagaraDataInterfaceSkeletalMesh
  &NiagaraDataInterfaceSocketReader
 ]
 usedDynamicInputs []
 usedModules [
  &self:SystemSpawnScript.SystemScriptSource.SystemScriptGraph.NiagaraNodeAssignment_1.SetVariables_#1
  &self:SystemSpawnScript.SystemScriptSource.SystemScriptGraph.NiagaraNodeAssignment_3.SetVariables_#2
  &self:SystemSpawnScript.SystemScriptSource.SystemScriptGraph.NiagaraNodeAssignment_4.SetVariables_#3
  &self:SystemSpawnScript.SystemScriptSource.SystemScriptGraph.NiagaraNodeAssignment_0.SetVariables_#4
  &self:SystemSpawnScript.SystemScriptSource.SystemScriptGraph.NiagaraNodeAssignment_2.SetVariables_#5
  &SystemState
  &Timeline
  &SampleSkeletalMesh
 ]
}
emitters {
 Muzzle_Inner_Flares {
  emitterUpdateScript [
   &EmitterState()
   &SpawnBurst_Instantaneous(Use Spawn Probability=false, UseLoopCountLimit=false)
   &SpawnRate(
    SpawnRate=&MakeFloatFromInt(Int=&MakeCustomIntFromBool(
      Bool=&IsPlatformSetActive(New Platform Set=DataInterface() {propertyValues {Platforms {qualityLevelMask=-8}}})
      True INT=10
      False INT=&MakeCustomIntFromBool(
       Bool=&IsPlatformSetActive(New Platform Set=DataInterface() {propertyValues {Platforms {qualityLevelMask=-28}}})
       True INT=9
       False INT=&MakeCustomIntFromBool(
        Bool=&IsPlatformSetActive(New Platform Set=DataInterface() {propertyValues {Platforms {qualityLevelMask=-30}}})
        True INT=7
        False INT=7
       )
      )
     ))
    Use Spawn Probability=false
   )
  ]
  particleSpawnScript [
   &InitializeParticle(
    Color=$User.PrimaryColor
    Lifetime Max=0.175
    Lifetime Min=0.075
    Lifetime Mode=::NewEnumerator1
    Sprite Size Mode=::NewEnumerator1
    Uniform Sprite Size Max=5000
    Uniform Sprite Size Min=2000
    UsePositionOffset=false
    Write Material Random=true
   )
   &CameraOffset(Camera Offset Amount=400)
   /SetVariables(
    Particles.Animorphic Color:LinearColor=<rgba 10 0.0508 7.7502 0.05>
    Particles.Animorphic Size:Vector2f=<1573.4686 67.9538>
    Particles.Soft Glow Size:Vector2f=<5000 2000>
   ) {isSetParameters=true}
  ]
  particleUpdateScript [
   &ParticleState()
   &ScaleColor(
    Scale Alpha=&FloatFromCurve(FloatCurve=DataInterface() {propertyValues {Curve {keys [
         {interpMode=RCIM_Cubic, tangentMode=RCTM_User, time=0, value=0}
         {interpMode=RCIM_Cubic, tangentMode=RCTM_User, time=0.5, value=1}
         {tangentMode=RCTM_User, time=1, value=0}
        ]}}})
    ScaleA=true
    ScaleRGB=true
   )
   /SetVariables(Particles.Soft Glow Color:LinearColor=&MakeLinearColorFromVectorAndFloat(
     "Vector (RGB)"=&MakeVectorFromLinearColorRGB(Color=<rgba 10 0.6856 1.7721 1>)
     "Float (Alpha)"=&FloatFromCurve(
      FloatCurve=DataInterface() {propertyValues {Curve {keys [
          {interpMode=RCIM_Cubic, tangentMode=RCTM_User, time=0, value=0}
          {interpMode=RCIM_Cubic, tangentMode=RCTM_User, time=0.5, value=1}
          {tangentMode=RCTM_User, time=1, value=0}
         ]}}}
      Scale Curve=&Multiply_Float(
       A=0.005
       B=&FloatFromCurve(
        FloatCurve=DataInterface() {propertyValues {Curve {keys [{interpMode=RCIM_Cubic, leaveTangent=-3.7201, tangentMode=RCTM_Break, time=0, value=1}, <at 1 0>]}}}
        CurveIndex=$System.TimeForDis
        Scale Curve=&FloatFromCurve(FloatCurve=DataInterface() {propertyValues {Curve {keys [<at 0.51 1>,<at 0.52 0>]}}}, CurveIndex=$User.EnergyBeamSwipe)
       )
      )
     )
    )) {isSetParameters=true}
   /SetVariables(
    Particles.Corona Color:LinearColor=&MakeLinearColorFromVectorAndFloat(
     "Vector (RGB)"=&MakeVectorFromLinearColorRGB(Color=<rgba 0.809 0.1336 1 1>)
     "Float (Alpha)"=&FloatFromCurve(
      FloatCurve=DataInterface() {propertyValues {Curve {keys [
          <at 0 0 RCIM_Cubic>
          {arriveTangent=2.5, interpMode=RCIM_Cubic, leaveTangent=2.5, time=0.165, value=0.65}
          {arriveTangent=0.1786, interpMode=RCIM_Cubic, leaveTangent=0.1786, time=0.4, value=1}
          {arriveTangent=-1.6667, interpMode=RCIM_Cubic, leaveTangent=-1.6667, time=0.725, value=0.75}
          <at 1 0 RCIM_Cubic>
         ]}}}
      Scale Curve=&Multiply_Float(
       A=0.5
       B=&FloatFromCurve(
        FloatCurve=DataInterface() {propertyValues {Curve {keys [{interpMode=RCIM_Cubic, leaveTangent=-3.7201, tangentMode=RCTM_Break, time=0, value=1}, <at 1 0>]}}}
        CurveIndex=$System.TimeForDis
        Scale Curve=&FloatFromCurve(FloatCurve=DataInterface() {propertyValues {Curve {keys [<at 0 1>,<at 0.51 1>,<at 0.52 0>]}}}, CurveIndex=$User.EnergyBeamSwipe)
       )
      )
     )
    )
    Particles.Corona Size:Vector2f=&Vector2DFromFloat(Value=&FloatFromCurve(
      FloatCurve=DataInterface() {propertyValues {Curve {keys [<at 0 1>,<at 0.5 0.7>,<at 1 1>]}}}
      Scale Curve=&Waveform(Global Amplitude Scale=2000, "Waveform [1]"=::NewEnumerator12, "[1] Phase Scale"=50, "[1] Frequency"=1, "[1] Amplitude Min/Max"=<0.85 1>)
     ))
   ) {isSetParameters=true}
   &DynamicMaterialParameters(
    Index 3 Param 1=&FloatFromCurve(FloatCurve=DataInterface() {propertyValues {Curve {keys [<at 0 0.7 RCIM_Cubic>,<at 0.5 0.25 RCIM_Cubic>,<at 1 0.7 RCIM_Cubic>]}}})
    Index 3 Param 2=1
    Index 3 Param 3=&RandomRangeFloat(Minimum=0.25, Maximum=0.35)
    Index 3 Param 4=1
    Module.Param0WriteEnabled3=true
    Module.Param1WriteEnabled3=true
    Module.Param2WriteEnabled3=true
    Module.Param3WriteEnabled3=true
    Write Parameter Index 0=false
    Write Parameter Index 3=true
   )
   /SetVariables() {isSetParameters=true}
   &ScaleColor(
    Scale Alpha=&FloatFromCurve(
     FloatCurve=DataInterface() {propertyValues {Curve {keys [{interpMode=RCIM_Cubic, leaveTangent=-3.7201, tangentMode=RCTM_Break, time=0, value=1}, <at 1 0>]}}}
     CurveIndex=$System.TimeForDis
     Scale Curve=&FloatFromCurve(FloatCurve=DataInterface() {propertyValues {Curve {keys [<at 0.51 1>,<at 0.52 0>]}}}, CurveIndex=$User.EnergyBeamSwipe)
    )
    ScaleA=true
    ScaleRGB=true
   ) {module=ScaleColor001}
   /SetVariables(Particles.Animorphic Color:LinearColor=&MultiplyLinearColor(
     Linear Color=$Particles.Initial.Animorphic Color
     Scale Factor=&MakeVector4FromVectorAndFloat(
      "Vector (RGB)"=<1 1 1>
      "Float (Alpha)"=&FloatFromCurve(
       FloatCurve=DataInterface() {propertyValues {Curve {keys [{interpMode=RCIM_Cubic, leaveTangent=-3.7201, tangentMode=RCTM_Break, time=0, value=1}, <at 1 0>]}}}
       CurveIndex=$System.TimeForDis
       Scale Curve=&FloatFromCurve(FloatCurve=DataInterface() {propertyValues {Curve {keys [<at 0.51 1>,<at 0.52 0>]}}}, CurveIndex=$User.EnergyBeamSwipe)
      )
     )
    )) {isSetParameters=true}
  ]
  rendererClasses [&NiagaraSpriteRendererProperties]
  renderers [
   {rendererClass=&NiagaraSpriteRendererProperties}
   {rendererClass=&NiagaraSpriteRendererProperties, rendererIndex=1}
   {rendererClass=&NiagaraSpriteRendererProperties, rendererIndex=2}
   {rendererClass=&NiagaraSpriteRendererProperties, rendererIndex=3}
  ]
 }
 NE_EnergyRibbons {
  emitterUpdateScript [
   &EmitterState(
    Inactive Response=::NewEnumerator1
    Life Cycle Mode=::NewEnumerator1
    Loop Behavior=::NewEnumerator1
    Loop Duration Mode=::NewEnumerator1
    UseLoopDelay=false
   )
   &SpawnBurst_Instantaneous(
    Spawn Count=&Multiply_Int(A=$System.RibbonNumber, B=$System.ParticlePerRibbon)
    Use Spawn Probability=false
    UseLoopCountLimit=false
   )
  ]
  particleSpawnScript [
   &InitializeParticle(Lifetime=5, Ribbon Width=40, Ribbon Width Mode=::NewEnumerator1, UsePositionOffset=false, Write Material Random=true)
   &PartitionParticles(
    ID to Partition:NiagaraInt32=&ReturnExecIndex()
    Normalized Index In Partition Includes One=true
    Number Of Partitions=$System.RibbonNumber
    Particles Per Partition:NiagaraInt32=$System.ParticlePerRibbon
    Partition Mode:ENiagara_PartitionMode_v2=::NewEnumerator1
    Write Partition Index to Ribbon Link Order:ENiagara_AttributeSamplingApplyOutput=::NewEnumerator0
    Write Partition to Ribbon ID=::NewEnumerator0
   )
   /SetVariables(Particles.RibbonU0Override=$Particles.PartitionParticles.Normalized Index In Partition) {isSetParameters=true, module=SetVariables_#6}
  ]
  particleUpdateScript [
   &ParticleState(Kill Particles When Lifetime Has Elapsed=false)
   &RotateAroundPoint(
    Radius=0
    Rotation Center:NiagaraPosition=$System.PortalPos
    Rotation Phase:NiagaraFloat=&Add_Float(
     A=&Multiply_Float(A=$Emitter.CustomTime, B=0)
     B=&Multiply_FloatByInt(
      Float=&Divide_Float(
       A=&RandomRangeFloat(Minimum=0, Maximum=1)
       B=&MakeFloatFromInt(Int=$System.RibbonNumber)
      )
      Integer=$Particles.PartitionParticles.Partition
     )
    )
    Rotation Rate=0.1
    Set Particle Position=false
    X Axis:Vector3f=<0 0 0>
    Y Axis:Vector3f=<1 1 0>
   )
   &ScaleRibbonWidth(Ribbon Width Scale=&Multiply_Float(
     A=&Lerp_Float(
      A=1
      B=5
      Alpha=&FloatFromCurve(
       FloatCurve=DataInterface() {propertyValues {Curve {keys [
           {arriveTangent=5.8422, interpMode=RCIM_Cubic, leaveTangent=5.8422, tangentMode=RCTM_User, time=0, value=1}
           {arriveTangent=0.0677, interpMode=RCIM_Cubic, leaveTangent=0.0677, tangentMode=RCTM_User, time=1, value=0.5}
          ]}}}
       CurveIndex=$Particles.PartitionParticles.Normalized Index In Partition
       Scale Curve=2
      )
     )
     B=$System.FadeOut
    ))
   /SetVariables(
    Particles.Position=&AddVectorToPosition(
     Position=&Lerp_Vector(
      A=&TransformPosition(Position=<0 0 0>, Source Space=::Local, Destination Space=::Local)
      B=&TransformPosition(Position=$User.Beam End, Source Space=::World, Destination Space=::Local, Apply Scale=true)
      Alpha=$Particles.Initial.PartitionParticles.Normalized Index In Partition
     )
     Vector=&Add_Vector(
      A=&MakeVector(
       X=&Waveform(
        Global Phase=&Add_Float(
         A=&Add_Float(A=$Particles.PartitionParticles.Normalized Index In Partition, B=$Emitter.Age)
         B=&RandomRangeFloat(Minimum=0, Maximum=1)
        )
        "Waveform [1]"=::NewEnumerator12
        "[1] Frequency"=1
        "[1] Amplitude Min/Max"=<-1 1>
        "[1] Scale"=&FloatFromCurve(
         FloatCurve=DataInterface() {propertyValues {Curve {
            defaultValue=1
            keys [
             {tangentMode=RCTM_User, time=0, value=1}
             {arriveTangent=-0.0405, interpMode=RCIM_Cubic, leaveTangent=-0.0405, tangentMode=RCTM_User, time=0.7009, value=1}
             {arriveTangent=-8.0089, interpMode=RCIM_Cubic, leaveTangent=-8.0089, tangentMode=RCTM_User, time=1, value=0}
            ]
           }}}
         CurveIndex=$Particles.PartitionParticles.Normalized Index In Partition
         Scale Curve=200
        )
       )
       Y=&Waveform(
        Global Phase=&Add_Float(
         A=&Add_Float(A=$Particles.PartitionParticles.Normalized Index In Partition, B=$Emitter.Age)
         B=&RandomRangeFloat(Minimum=0, Maximum=1)
        )
        "Waveform [1]"=::NewEnumerator12
        "[1] Frequency"=1
        "[1] Amplitude Min/Max"=<-1 1>
        "[1] Scale"=&FloatFromCurve(
         FloatCurve=DataInterface() {propertyValues {Curve {
            defaultValue=1
            keys [
             {tangentMode=RCTM_User, time=0, value=1}
             {arriveTangent=-0.0405, interpMode=RCIM_Cubic, leaveTangent=-0.0405, tangentMode=RCTM_User, time=0.7009, value=1}
             {arriveTangent=-8.0089, interpMode=RCIM_Cubic, leaveTangent=-8.0089, tangentMode=RCTM_User, time=1, value=0}
            ]
           }}}
         CurveIndex=$Particles.PartitionParticles.Normalized Index In Partition
         Scale Curve=200
        )
       )
       Z=&Waveform(
        Global Phase=&Add_Float(
         A=&Add_Float(A=$Particles.PartitionParticles.Normalized Index In Partition, B=$Emitter.Age)
         B=&RandomRangeFloat(Minimum=0, Maximum=1)
        )
        "Waveform [1]"=::NewEnumerator12
        "[1] Frequency"=5
        "[1] Amplitude Min/Max"=<-1 1>
        "[1] Scale"=&FloatFromCurve(
         FloatCurve=DataInterface() {propertyValues {Curve {
            defaultValue=1
            keys [
             {tangentMode=RCTM_User, time=0, value=1}
             {arriveTangent=-0.0405, interpMode=RCIM_Cubic, leaveTangent=-0.0405, tangentMode=RCTM_User, time=0.7009, value=1}
             {arriveTangent=-8.0089, interpMode=RCIM_Cubic, leaveTangent=-8.0089, tangentMode=RCTM_User, time=1, value=0}
            ]
           }}}
         CurveIndex=$Particles.PartitionParticles.Normalized Index In Partition
         Scale Curve=200
        )
       )
      )
      B=&Multiply_VectorByFloat(
       Vector=/Niagara/DynamicInputs/Math/Clamp_Vector(
        Vector:Vector3f=/Niagara/DynamicInputs/Vectors/RotateVector(
         Vector To Rotate:Vector3f=&RandomRangeVector(Minimum=<0 0 1>, Maximum=<0 0 1>)
         Yaw:NiagaraFloat=1
         Pitch:NiagaraFloat=1
         Roll:NiagaraFloat=1
         Rotation Rate:NiagaraFloat=$Engine.DeltaTime
        )
        Min:Vector3f=<0 0 0>
        Max:Vector3f=<0 0 0.1>
       )
       Float=&FloatFromCurve(
        FloatCurve=DataInterface() {propertyValues {Curve {keys [
            {arriveTangent=1115.3998, interpMode=RCIM_Cubic, leaveTangent=1115.3998, tangentMode=RCTM_User, time=0, value=0}
            {
             arriveTangent=-429.2098
             arriveTangentWeight=0.1667
             interpMode=RCIM_Cubic
             leaveTangent=-429.2098
             leaveTangentWeight=0.1667
             tangentMode=RCTM_User
             time=0.8
             value=225
            }
            {arriveTangent=-824.957, interpMode=RCIM_Cubic, leaveTangent=-824.957, tangentMode=RCTM_User, time=1, value=0}
           ]}}}
        CurveIndex=$Particles.PartitionParticles.Normalized Index In Partition
        Scale Curve=&RandomRangeFloat(Minimum=150, Maximum=120)
       )
      )
     )
    )
    Particles.RandomOffset:NiagaraFloat=&Add_Float(
     A=&RandomRangeFloat(Minimum=0, Maximum=1)
     B=&Multiply_Float(A=$Engine.System.Age, B=1)
    )
   ) {isSetParameters=true, module=SetVariables_#7}
   &DynamicMaterialParameters(
    Index 0 Param 1=1
    Index 0 Param 2=0.05
    Index 0 Param 3=$Particles.RandomOffset
    Index 0 Param 4=1
    Module.Index0Float0Write=true
    Module.Index0Float1Write=true
    Module.Index0Float2Write=true
    Module.Index0Float3Write=true
    Module.Index1Float0Write=true
    Module.Index1Float1Write=true
    Module.Index1Float2Write=true
    Module.Index1Float3Write=true
    Write Parameter Index 1=::NewEnumerator1
   ) {module=DynamicMaterialParameters001}
   /SetVariables(
    Particles.GlowColor=&MultiplyLinearColor(Linear Color=$Particles.Color, Scale Factor=<1 1 1 0.1>)
    Particles.GlowDynPar:Vector4f=&MakeVector4(X=1, Y=1, Z=$Particles.RandomOffset, W=1)
    Particles.GlowWidth:NiagaraFloat=&Multiply_Float(A=$Particles.RibbonWidth, B=1.65)
    Particles.emptytohalf:NiagaraFloat=/Niagara/DynamicInputs/Math/Floor_Float(Float:NiagaraFloat=&ClampFloat(Float=&Add_Float(
       A=&OneMinusFloat(Float=$Particles.PartitionParticles.Normalized Index In Partition)
       B=&FloatFromCurve(FloatCurve=DataInterface() {propertyValues {Curve {keys [<at 0 0>,<at 0.5 1>]}}}, CurveIndex=$User.EnergyBeamSwipe)
      )))
    Particles.halftofull:NiagaraFloat=&Ceil_Float(Float=&ClampFloat(Float=&Add_Float(
       A=$Particles.PartitionParticles.Normalized Index In Partition
       B=&FloatFromCurve(FloatCurve=DataInterface() {propertyValues {Curve {keys [<at 0.5 0>,<at 1 -1>]}}}, CurveIndex=$User.EnergyBeamSwipe)
      )))
   ) {isSetParameters=true, module=SetVariables_#8}
   &Color(Color=&MultiplyLinearColor(
     Linear Color=&ColorFromCurve(
      ColorCurve=DataInterface() {propertyValues {
        AlphaCurve {keys [<at 0.9513 1 RCIM_Cubic>,<at 1 0>]}
        BlueCurve {keys [<at 0.001 1>,<at 0.5597 0.8668>,<at 0.9732 2 RCIM_Cubic>]}
        GreenCurve {keys [<at 0.001 0.2845>,<at 0.5597 0.2101>,<at 0.9732 2 RCIM_Cubic>]}
        RedCurve {keys [<at 0.001 0.5657>,<at 0.5597 1>,<at 0.9732 2 RCIM_Cubic>]}
       }}
      CurveIndex=$Particles.PartitionParticles.Normalized Index In Partition
     )
     Scale Factor=&MakeVector4FromVectorAndFloat(
      "Vector (RGB)"=&VectorFromFloat(Value=&Lerp_Float(A=1, B=3, Alpha=$Particles.PartitionParticles.Normalized Index In Partition))
      "Float (Alpha)"=&Lerp_Float(
       A=$Particles.emptytohalf
       B=$Particles.halftofull
       Alpha=&FloatFromCurve(FloatCurve=DataInterface() {propertyValues {Curve {keys [<at 0 0 RCIM_Constant>,<at 0.5 1 RCIM_Constant>]}}})
      )
     )
    )) {module=Color001}
  ]
  rendererClasses [&NiagaraRibbonRendererProperties]
  renderers [
   {rendererClass=&NiagaraRibbonRendererProperties}
   {rendererClass=&NiagaraRibbonRendererProperties, rendererIndex=1}
  ]
 }
 NE_EnergyRibbons001 {
  emitterSpawnScript [/SetVariables() {isSetParameters=true}]
  emitterUpdateScript [
   /SetVariables(Emitter.CustomTime:NiagaraFloat=&Add_Float(
     A=$Emitter.CustomTime
     B=&Multiply_Float(A=$Engine.DeltaTime, B=&Lerp_Float(A=1, B=3, Alpha=$System.Oscillation))
    )) {isSetParameters=true}
   &EmitterState(
    Inactive Response=::NewEnumerator1
    Life Cycle Mode=::NewEnumerator1
    Loop Behavior=::NewEnumerator1
    Loop Duration Mode=::NewEnumerator1
    UseLoopDelay=false
   )
   &SpawnBurst_Instantaneous(Spawn Count=&Multiply_Int(A=1, B=4), Use Spawn Probability=false, UseLoopCountLimit=false)
  ]
  particleSpawnScript [
   &PartitionParticles(
    ID to Partition:NiagaraInt32=&ReturnExecIndex()
    Normalized Index In Partition Includes One=true
    Number Of Partitions=$System.RibbonNumber
    Particles Per Partition:NiagaraInt32=$System.ParticlePerRibbon
    Partition Mode:ENiagara_PartitionMode_v2=::NewEnumerator1
    Write Partition Index to Ribbon Link Order:ENiagara_AttributeSamplingApplyOutput=::NewEnumerator0
    Write Partition to Ribbon ID=::NewEnumerator0
   )
   &InitializeParticle(
    Lifetime=5
    Ribbon Width=&Lerp_Float(A=12.5, B=3, Alpha=$Particles.PartitionParticles.Normalized Index In Partition)
    Ribbon Width Mode=::NewEnumerator1
    UsePositionOffset=false
    Write Material Random=true
   )
   /SetVariables(Particles.RibbonU0Override=$Particles.PartitionParticles.Normalized Index In Partition) {isSetParameters=true, module=SetVariables_#6}
  ]
  particleUpdateScript [
   &ParticleState(Kill Particles When Lifetime Has Elapsed=false)
   &RotateAroundPoint(
    Radius=0
    Rotation Center:NiagaraPosition=$System.PortalPos
    Rotation Phase:NiagaraFloat=&Add_Float(
     A=&Multiply_Float(A=$Emitter.CustomTime, B=16400.6543)
     B=&Multiply_FloatByInt(
      Float=&Divide_Float(
       A=&RandomRangeFloat(Minimum=0, Maximum=1)
       B=&MakeFloatFromInt(Int=$System.RibbonNumber)
      )
      Integer=$Particles.PartitionParticles.Partition
     )
    )
    Rotation Rate=0.2
    Set Particle Position=false
    X Axis:Vector3f=<0 0 0>
    Y Axis:Vector3f=<1 1 0>
   )
   &ScaleRibbonWidth(Ribbon Width Scale=&Multiply_Float(A=50, B=$System.FadeOut))
   /SetVariables(Particles.Position=&AddVectorToPosition(
     Position=&Lerp_Vector(
      A=&TransformPosition(Position=<0 0 0>, Source Space=::Local, Destination Space=::Local)
      B=&TransformPosition(Position=$User.Beam End, Source Space=::World, Destination Space=::Local, Apply Scale=true)
      Alpha=$Particles.Initial.PartitionParticles.Normalized Index In Partition
     )
     Vector=&Add_Vector(
      A=&RandomRangeVector(Minimum=<0 0 0>, Maximum=<0 0 0>)
      B=&Multiply_VectorByFloat(
       Vector=&Clamp_Vector(
        Vector:Vector3f=&RotateVector(
         Vector To Rotate:Vector3f=&RandomRangeVector(Minimum=<0 0 1>, Maximum=<0 0 1>)
         Yaw:NiagaraFloat=1
         Pitch:NiagaraFloat=1
         Roll:NiagaraFloat=1
         Rotation Rate:NiagaraFloat=$Engine.DeltaTime
        )
        Min:Vector3f=<0 0 0>
        Max:Vector3f=<0 0 0.1>
       )
       Float=&FloatFromCurve(
        FloatCurve=DataInterface() {propertyValues {Curve {keys [
            {arriveTangent=1115.3998, interpMode=RCIM_Cubic, leaveTangent=1115.3998, tangentMode=RCTM_User, time=0, value=0}
            {
             arriveTangent=-468.9583
             arriveTangentWeight=0.1667
             interpMode=RCIM_Cubic
             leaveTangent=-468.9583
             leaveTangentWeight=0.1667
             tangentMode=RCTM_User
             time=0.8
             value=225
            }
            {arriveTangent=-1058.5016, interpMode=RCIM_Cubic, leaveTangent=-1058.5016, tangentMode=RCTM_User, time=1, value=0}
           ]}}}
        CurveIndex=$Particles.PartitionParticles.Normalized Index In Partition
        Scale Curve=&RandomRangeFloat(Minimum=130, Maximum=130)
       )
      )
     )
    )) {isSetParameters=true, module=SetVariables_#7}
   &DynamicMaterialParameters(
    Index 0 Param 1=&Multiply_FloatByInt(Float=1, Integer=$Particles.PartitionParticles.Partition)
    Index 1 Param 2=$User.EnergyBeamSwipe
    Module.Index0Float0Write=true
    Module.Index0Float1Write=true
    Module.Index0Float2Write=true
    Module.Index0Float3Write=true
    Module.Index1Float0Write=true
    Module.Index1Float1Write=true
    Module.Index1Float2Write=true
    Module.Index1Float3Write=true
    Write Parameter Index 1=::NewEnumerator1
   )
   &Color(
    Color=&ColorFromCurve(
     ColorCurve=DataInterface() {propertyValues {
       AlphaCurve {keys [<at 0 1 RCIM_Cubic>]}
       BlueCurve {keys [<at 0 0.1348>,<at 0.4345 0.0274>,<at 0.9891 1 RCIM_Cubic>]}
       GreenCurve {keys [<at 0 0.0509>,<at 0.4345 0.0035>,<at 0.9891 1 RCIM_Cubic>]}
       RedCurve {keys [<at 0 0.25>,<at 0.4345 1>,<at 0.9891 1 RCIM_Cubic>]}
      }}
     CurveIndex=$Particles.PartitionParticles.Normalized Index In Partition
    )
    Scale Alpha=2
    Scale Color=<50 50 50>
   )
   /SetVariables(
    Particles.GlowColor=&MultiplyLinearColor(Linear Color=$Particles.Color, Scale Factor=<1 1 1 0.1>)
    Particles.GlowWidth:NiagaraFloat=&Multiply_Float(A=$Particles.RibbonWidth, B=1)
   ) {isSetParameters=true, module=SetVariables_#8}
  ]
  rendererClasses [&NiagaraRibbonRendererProperties]
  renderers [{rendererClass=&NiagaraRibbonRendererProperties}]
 }
}
summary {
 systemName=NS_PistolCanary_MM_DrainEnergy_Beam
 emitters [
  {emitterName=NE_EnergyRibbons, rendererClasses [&NiagaraRibbonRendererProperties]}
  {emitterName=Muzzle_Inner_Flares, rendererClasses [&NiagaraSpriteRendererProperties]}
  {emitterName=NE_EnergyRibbons001, rendererClasses [&NiagaraRibbonRendererProperties]}
 ]
 userVariables [
  User.Alpha_Index 1 Param 2 {type=NiagaraFloat, defaultValue=0, description=""}
  User.Beam End {type=NiagaraPosition, description="", defaultValue {}}
  User.EnergyBeamSwipe {type=NiagaraFloat, defaultValue=0.75, description=""}
  User.FadeOut {type=NiagaraFloat, defaultValue=1, description=""}
  User.HandPosOffset {type=Vector3f, description="", defaultValue {struct=Vector3f, value=<0 0 0>}}
  User.PrimaryColor {type=LinearColor, description="", defaultValue {struct=LinearColor, value=<rgba 1 0.0511 0.1756 1>}}
  User.Size {type=NiagaraFloat, defaultValue=1, description=""}
  User.Stop Location {type=NiagaraPosition, description="", defaultValue {}}
 ]
}
