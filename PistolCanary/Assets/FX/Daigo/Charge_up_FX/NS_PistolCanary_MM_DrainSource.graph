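# Niagara system for the source of a PistolCanary drain charge-up effect, with only the glow and lightning emitters enabled.
# NE_BaseGlow bursts persistent sprites and lights, including one offset from the portal position, with color and intensity driven by oscillation and fade-out values.
# NE_Lightnings spawns short-lived, material-animated meshes at random sockets, scaling rate by fade-out and size by build-up.
# Disabled smoke, flare and spark emitters read particle positions from the disabled spiraling energy ribbons.
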NiagaraSystem /PistolCanary/Assets/FX/Daigo/Charge_up_FX/NS_PistolCanary_MM_DrainSource
dependencies {
 usedDataInterfaces [
  &NiagaraDataInterfaceColorCurve
  &NiagaraDataInterfaceParticleRead
  &NiagaraDataInterfaceArrayFloat
  &NiagaraDataInterfaceArrayColor
  &NiagaraDataInterfaceSpriteRendererInfo
  &NiagaraDataInterfaceCurve
  &NiagaraDataInterfaceSkeletalMesh
  &NiagaraDataInterfaceSocketReader
 ]
 usedDynamicInputs [&Vector2DFromFloat]
 usedModules [
  &self:SystemSpawnScript.SystemScriptSource.SystemScriptGraph.NiagaraNodeAssignment_1.SetVariables_#1
  &self:SystemSpawnScript.SystemScriptSource.SystemScriptGraph.NiagaraNodeAssignment_3.SetVariables_#2
  &self:SystemSpawnScript.SystemScriptSource.SystemScriptGraph.NiagaraNodeAssignment_4.SetVariables_#3
  &self:SystemSpawnScript.SystemScriptSource.SystemScriptGraph.NiagaraNodeAssignment_5.SetVariables_#4
  &self:SystemSpawnScript.SystemScriptSource.SystemScriptGraph.NiagaraNodeAssignment_6.SetVariables_#5
  &self:SystemSpawnScript.SystemScriptSource.SystemScriptGraph.NiagaraNodeAssignment_0.SetVariables_#6
  &self:SystemSpawnScript.SystemScriptSource.SystemScriptGraph.NiagaraNodeAssignment_2.SetVariables_#7
  &SystemState
  &Timeline
  &SampleSkeletalMesh
 ]
}
emitters {
 NE_BaseGlow {
  emitterUpdateScript [
   &EmitterState(
    Inactive Response=::NewEnumerator1
    Life Cycle Mode=::NewEnumerator1
    Loop Behavior=::NewEnumerator1
    Loop Duration Mode=::NewEnumerator1
    UseLoopDelay=false
   )
   &SpawnBurst_Instantaneous(Spawn Count=&Add_Integer(A=1, B=1), Use Spawn Probability=false, UseLoopCountLimit=false)
   &SpawnBurst_Instantaneous(Spawn Group=1, Use Spawn Probability=false, UseLoopCountLimit=false) {module=SpawnBurst_Instantaneous001}
  ]
  particleSpawnScript [
   /SetVariables(Particles.VisibilityTag=&SpawnGroup()) {isSetParameters=true}
   &PartitionParticles(Number Of Partitions=&Add_Integer(A=$System.RibbonNumber, B=1)) {enabled=false}
   &InitializeParticle(
    Color Mode=::NewEnumerator0
    Lifetime=5
    Sprite Size=/Niagara/DynamicInputs/Helpers/MaskVector2d(
     Mask Source:NiagaraInt32=0
     Mask Parameter:NiagaraInt32=$Particles.PartitionParticles.Partition
     Mask:NiagaraInt32=1
     Masked Vector2D:Vector2f=&Multiply_Vector2D_ByFloat(Float=1, Vector2D=<150 150>)
     Passthrough Vector2D:Vector2f=&Multiply_Vector2D_ByFloat(Float=5, Vector2D=<200 200>)
    )
    Sprite Size Mode=::NewEnumerator4
    UsePositionOffset=false
    Write Material Random=true
   )
   &CameraOffset(Camera Offset Amount=5)
  ]
  particleUpdateScript [
   &ParticleState(Kill Particles When Lifetime Has Elapsed=false)
   &Color(Color=&Multiply_LinearColorByFloat(
     Linear Color=&MaskLinearColor(
      Mask Parameter=$Particles.PartitionParticles.Partition
      Mask=0
      Masked Linear Color=&ColorFromCurve(
       ColorCurve=DataInterface() {propertyValues {
         AlphaCurve {keys [<at 0 1 RCIM_Cubic>]}
         BlueCurve {keys [<at 0 4.2296 RCIM_Cubic>,<at 1 1 RCIM_Cubic>]}
         GreenCurve {keys [<at 0 0 RCIM_Cubic>,<at 1 0.1291 RCIM_Cubic>]}
         RedCurve {keys [<at 0 5 RCIM_Cubic>,<at 1 0.5263 RCIM_Cubic>]}
        }}
       CurveIndex=$System.Oscillation
      )
      Passthrough Linear Color=&ColorFromCurve(
       ColorCurve=DataInterface() {propertyValues {
         AlphaCurve {keys [<at 0 1 RCIM_Cubic>]}
         BlueCurve {keys [<at 0 0.1338 RCIM_Cubic>,<at 1 0.13 RCIM_Cubic>]}
         GreenCurve {keys [<at 0 0 RCIM_Cubic>,<at 1 0.0005 RCIM_Cubic>]}
         RedCurve {keys [<at 0 1 RCIM_Cubic>,<at 1 0.3906 RCIM_Cubic>]}
        }}
       CurveIndex=$System.Oscillation
      )
     )
     Float=$System.FadeOut
    ))
   &self:FollowRibbon(
    Alpha=0.999
    ParticlePerRibbon:NiagaraInt32=$System.ParticlePerRibbon
    RibbonNumber:NiagaraInt32=&Add_Integer(A=$Particles.PartitionParticles.Partition, B=-1)
    RibbonReader=DataInterface() {propertyValues {EmitterBinding {bindingMode=Other, emitterName=NE_EnergyRibbons}}}
   ) {enabled=false, module=FollowRibbon}
   /SetVariables(Particles.Position=/Niagara/DynamicInputs/Helpers/MaskPosition(
     Mask Parameter:NiagaraInt32=$Particles.PartitionParticles.Partition
     Mask:NiagaraInt32=0
     Masked Position:NiagaraPosition=&TransformPosition(Position=$System.SampleSkeletalMesh.SampledPosition, Source Space=::World, Destination Space=::Local)
     Passthrough Position:NiagaraPosition=$Output.FollowRibbon.RibbonPos
    )) {enabled=false, isSetParameters=true}
   /SetVariables(Particles.LongFlareSize:Vector2f=&Multiply_Vector2D(
     A=$Particles.SpriteSize
     B=&MaskVector2d(
      Mask Parameter:NiagaraInt32=$Emitter.SpawnGroup
      Mask:NiagaraInt32=$Particles.PartitionParticles.Partition
      Masked Vector2D:Vector2f=<3 0.5>
      Passthrough Vector2D:Vector2f=<3 3>
     )
    )) {isSetParameters=true}
   &FN_Sprite_Attributes(
    Color=&MultiplyLinearColor(Linear Color=$Particles.Color, Scale Factor=&MakeVector4_FromLinearColor(LinearColor=<rgba 1 1 1 0.1>))
    Create Color Attribute=true
    Create SpriteSize Attribute=true
    SpriteSize=&Multiply_Vector2D_ByFloat(Float=5, Vector2D=$Particles.SpriteSize)
    Write to Module Namespace=true
   )
   &FN_Sprite_Attributes(
    Color=&MultiplyLinearColor(
     Linear Color=$Particles.Color
     Scale Factor=&MakeVector4_FromLinearColor(LinearColor=<rgba 0.02 0.02 0.02 1>)
    )
    Create Color Attribute=true
    Create SpriteSize Attribute=true
    SpriteSize=&Multiply_Vector2D_ByFloat(Float=1, Vector2D=$Particles.SpriteSize)
    Write to Module Namespace=true
   ) {module=FN_Sprite_Attributes001}
   &SpriteFacingAndAlignment(Sprite Facing=<0 -1 0>, WriteAlignment=false, WriteFacing=false)
   &FN_Light_Attributes(
    Color=$Particles.Color
    Create Color Attribute=true
    Create Exponent Attribute=true
    Create Radius Attribute=true
    Enable Particle Light Enable=true
    Exponent=1
    Light Enabled:NiagaraBool=&SetBoolByIntComparison(A=$Particles.PartitionParticles.Partition, B=1, Comparison Type=::NewEnumerator6)
    Radius=150
    Scale Color=5
   )
   &FN_Light_Attributes(
    Color=$Particles.Color
    Create Color Attribute=true
    Create Exponent Attribute=true
    Create Position Attribute=true
    Create Radius Attribute=true
    Enable Particle Light Enable=true
    Exponent=1
    Light Enabled:NiagaraBool=&SetBoolByIntComparison(A=&SpawnGroup(), B=1, Comparison Type=::NewEnumerator4)
    Position=&AddVectorToPosition(Position=Linked() {linkedVariable {name=System.PortalPos, type=Vector3f}}, Vector=<0 -250 0>)
    Radius=550
    Scale Color=&Multiply_Float(A=&Lerp_Float(A=5, B=10, Alpha=$System.Oscillation), B=$System.FadeOut)
   ) {module=FN_Light_Attributes001}
   &FN_Mesh_Attributes(
    Color=&Multiply_LinearColorByFloat(Linear Color=<rgba 10 0.942 4.4903 1>, Float=$System.FadeOut)
    Create Color Attribute=true
    Create Orientation Attribute=true
    Create Scale Attribute=true
    Orientation:Quat4f=/Niagara/DynamicInputs/Helpers/MaskQuaternion(
     Mask Parameter:NiagaraInt32=$Particles.PartitionParticles.Partition
     Mask:NiagaraInt32=0
     Masked Quaternion:Quat4f=&OrientMeshToVector(
      Look At Direction=&NormalizeVector(
       Vector To Normalize=&Subtract_Vector(A=/Niagara/DynamicInputs/Helpers/SimulationPosition(), B=$System.PortalPos)
       Safe Normalize=true
       Safe Fallback Threshold=0
       Safe Fallback Vector=<0 0 0.001>
      )
      Reference Axis Vector=<1 0 0>
     )
     Passthrough Quaternion:Quat4f=&OrientMeshToVector(
      Look At Direction=&Multiply_VectorByFloat(Vector=$Output.FollowRibbon.DirectionVector, Float=1)
      Reference Axis Vector=<1 0 0>
     )
    )
    Scale=&MaskVector(Mask Parameter=$Particles.PartitionParticles.Partition, Mask=0, Masked Vector=<2 0.4 0.4>, Passthrough Vector=<2 0.3 0.3>)
    Write to Module Namespace=true
   ) {enabled=false}
   &DynamicMaterialParameters(
    Index 3 Param 1=1
    Index 3 Param 3=&Multiply_FloatByInt(Float=0.33, Integer=$Particles.PartitionParticles.Partition)
    Index 3 Param 4=&Multiply_Float(
     A=$Engine.System.Age
     B=&MaskFloat(Mask Parameter=$Particles.PartitionParticles.Partition, Mask=0, Masked Float=1, Passthrough Float=-1)
    )
    Module.Index3Float0Write=true
    Module.Index3Float1Write=true
    Module.Index3Float2Write=true
    Module.Index3Float3Write=true
    Write Parameter Index 0=::NewEnumerator0
    Write Parameter Index 3=::NewEnumerator1
   )
  ]
  rendererClasses [
   &NiagaraSpriteRendererProperties
   &NiagaraLightRendererProperties
   &NiagaraMeshRendererProperties
  ]
  renderers [
   {rendererClass=&NiagaraSpriteRendererProperties}
   {rendererClass=&NiagaraSpriteRendererProperties, rendererIndex=1}
   {rendererClass=&NiagaraSpriteRendererProperties, rendererIndex=2}
   {rendererClass=&NiagaraLightRendererProperties, rendererIndex=3}
   {rendererClass=&NiagaraMeshRendererProperties, rendererIndex=4}
   {rendererClass=&NiagaraLightRendererProperties, rendererIndex=5}
  ]
 }
 NE_EnergyRibbons {
  bEnabled=false
  emitterSpawnScript [/SetVariables() {isSetParameters=true}]
  emitterUpdateScript [
   /SetVariables(Emitter.CustomTime:NiagaraFloat=&Add_Float(
     A=$Emitter.CustomTime
     B=&Multiply_Float(A=$Engine.DeltaTime, B=&Lerp_Float(A=1, B=3, Alpha=$System.Oscillation))
    )) {isSetParameters=true}
   &EmitterState(
    Inactive Response=::NewEnumerator1
    Life Cycle Mode=::NewEnumerator1
    Loop Behavior=::NewEnumerator1
    Loop Duration Mode=::NewEnumerator1
    UseLoopDelay=false
   )
   &SpawnBurst_Instantaneous(
    Spawn Count=&Multiply_Int(A=$System.RibbonNumber, B=$System.ParticlePerRibbon)
    Use Spawn Probability=false
    UseLoopCountLimit=false
   )
  ]
  particleSpawnScript [
   &InitializeParticle(Lifetime=5, Ribbon Width=30, Ribbon Width Mode=::NewEnumerator1, UsePositionOffset=false, Write Material Random=true)
   &PartitionParticles(
    ID to Partition:NiagaraInt32=&ReturnExecIndex()
    Normalized Index In Partition Includes One=true
    Number Of Partitions=$System.RibbonNumber
    Particles Per Partition:NiagaraInt32=$System.ParticlePerRibbon
    Partition Mode:ENiagara_PartitionMode=::NewEnumerator1
    Write Partition Index to Ribbon Link Order:ENiagara_AttributeSamplingApplyOutput=::NewEnumerator0
    Write Partition to Ribbon ID=::NewEnumerator0
   )
   /SetVariables(Particles.RibbonU0Override=$Particles.PartitionParticles.Normalized Index In Partition) {isSetParameters=true}
  ]
  particleUpdateScript [
   &ParticleState(Kill Particles When Lifetime Has Elapsed=false)
   &RotateAroundPoint(
    Radius=110
    Rotation Center:NiagaraPosition=$System.PortalPos
    Rotation Phase:NiagaraFloat=&Add_Float(
     A=&Multiply_Float(A=$Emitter.CustomTime, B=0.15)
     B=&Multiply_FloatByInt(
      Float=&Divide_Float(A=1, B=&MakeFloatFromInt(Int=$System.RibbonNumber))
      Integer=$Particles.PartitionParticles.Partition
     )
    )
    Set Particle Position=false
    X Axis:Vector3f=<0 0 1>
    Y Axis:Vector3f=<1 0 0>
   )
   &ScaleRibbonWidth(Ribbon Width Scale=&Multiply_Float(
     A=&Lerp_Float(
      A=1
      B=5
      Alpha=&FloatFromCurve(
       FloatCurve=DataInterface() {propertyValues {Curve {keys [
           {arriveTangent=15.0632, interpMode=RCIM_Cubic, leaveTangent=15.0632, tangentMode=RCTM_User, time=0, value=0.2}
           {arriveTangent=0.0677, interpMode=RCIM_Cubic, leaveTangent=0.0677, tangentMode=RCTM_User, time=1, value=0.2}
          ]}}}
       CurveIndex=$Particles.PartitionParticles.Normalized Index In Partition
      )
     )
     B=$System.FadeOut
    ))
   /SetVariables(Particles.SpiralAlpha:NiagaraFloat=&Add_Float(
     A=&Add_Float(
      A=&Multiply_Float(A=$Emitter.CustomTime, B=0.5)
      B=&Multiply_Float(A=$Particles.PartitionParticles.Normalized Index In Partition, B=1.5)
     )
     B=&Multiply_FloatByInt(
      Float=&Divide_Float(A=1, B=&MakeFloatFromInt(Int=$System.RibbonNumber))
      Integer=$Particles.PartitionParticles.Partition
     )
    )) {isSetParameters=true}
   /SetVariables(Particles.SpiralOffset:Vector3f=&Multiply_VectorByFloat(
     Vector=&MakeVector(
      X=&Cosine(Normalized Angle=$Particles.SpiralAlpha, Period=1, Scale=1, Bias=0)
      Y=0
      Z=&Sine(Normalized Angle=$Particles.SpiralAlpha, Period=1, Scale=1, Bias=0)
     )
     Float=&Multiply_Float(
      A=100
      B=&FloatFromCurve(
       FloatCurve=DataInterface() {propertyValues {Curve {keys [
           {arriveTangent=1.1062, interpMode=RCIM_Cubic, leaveTangent=1.1062, tangentMode=RCTM_User, time=0, value=0}
           {arriveTangent=0.6991, interpMode=RCIM_Cubic, leaveTangent=0.6991, tangentMode=RCTM_User, time=0.8, value=1}
           {arriveTangent=-3.046, interpMode=RCIM_Cubic, leaveTangent=-3.046, tangentMode=RCTM_User, time=1, value=0}
          ]}}}
       CurveIndex=$Particles.PartitionParticles.Normalized Index In Partition
      )
     )
    )) {isSetParameters=true}
   /SetVariables(Particles.Position=&AddVectorToPosition(
     Position=&Lerp_Vector(
      A=&TransformPosition(Position=<0 0 0>, Source Space=::Local, Destination Space=::Local)
      B=&TransformPosition(Position=<11980 79912 8444>, Source Space=::World, Destination Space=::Local)
      Alpha=$Particles.Initial.PartitionParticles.Normalized Index In Partition
     )
     Vector=&Add_Vector(
      A=&TransformPosition(Position=<0 0 0>, Source Space=::Local, Destination Space=::Local)
      B=&Multiply_VectorByFloat(
       Vector=/Niagara/DynamicInputs/Math/Clamp_Vector(Vector:Vector3f=$Particles.RotateAroundPoint.VectorFromCenter, Min:Vector3f=<-1 -1 0>, Max:Vector3f=<1 1 0>)
       Float=&FloatFromCurve(
        FloatCurve=DataInterface() {propertyValues {Curve {keys [
            {arriveTangent=-0.0375, interpMode=RCIM_Cubic, leaveTangent=-0.0375, tangentMode=RCTM_User, time=0, value=0}
            {arriveTangent=-1.3461, interpMode=RCIM_Cubic, leaveTangent=-1.3461, tangentMode=RCTM_User, time=0.7, value=1}
            {arriveTangent=-1.4067, interpMode=RCIM_Cubic, leaveTangent=-1.4067, tangentMode=RCTM_User, time=1, value=0}
           ]}}}
        CurveIndex=$Particles.PartitionParticles.Normalized Index In Partition
        Scale Curve=200
       )
      )
     )
    )) {isSetParameters=true}
   &DynamicMaterialParameters(
    Index 0 Param 1=&Multiply_FloatByInt(Float=0.3, Integer=$Particles.PartitionParticles.Partition)
    Module.Index0Float0Write=true
    Module.Index0Float1Write=true
    Module.Index0Float2Write=true
    Module.Index0Float3Write=true
   )
   &Color(Color=&ColorFromCurve(
     ColorCurve=DataInterface() {propertyValues {
       AlphaCurve {keys [<at 0 1 RCIM_Cubic>]}
       BlueCurve {keys [<at 0.2463 1 RCIM_Cubic>,<at 1 0.0186>]}
       GreenCurve {keys [<at 0.2463 1 RCIM_Cubic>,<at 1 0>]}
       RedCurve {keys [<at 0.2463 1 RCIM_Cubic>,<at 1 1>]}
      }}
     CurveIndex=$Particles.PartitionParticles.Normalized Index In Partition
    ))
   /SetVariables(
    Particles.GlowColor=&MultiplyLinearColor(Linear Color=$Particles.Color, Scale Factor=<1 1 1 0.2>)
    Particles.GlowWidth:NiagaraFloat=&Multiply_Float(A=$Particles.RibbonWidth, B=3)
   ) {isSetParameters=true}
  ]
  rendererClasses [&NiagaraRibbonRendererProperties]
  renderers [
   {rendererClass=&NiagaraRibbonRendererProperties}
   {rendererClass=&NiagaraRibbonRendererProperties, rendererIndex=1}
  ]
 }
 NE_Flare {
  bEnabled=false
  emitterUpdateScript [
   &EmitterState() {module=EmitterState001}
   &SpawnRate(SpawnGroup=1, SpawnRate=&Multiply_Float(A=3, B=$System.FadeOut), Use Spawn Probability=false)
   &SpawnRate(SpawnRate=&Multiply_Float(A=15, B=$System.FadeOut), Use Spawn Probability=false) {module=SpawnRate001}
  ]
  particleSpawnScript [
   &InitializeParticle(
    Lifetime=&MaskFloatBySpawnGroup(Masked Float=0.7, Passthrough Float=1.5)
    Mass Mode=::NewEnumerator1
    Mesh Scale Mode=::NewEnumerator3
    Mesh Uniform Scale=&MaskFloat(Mask Parameter=$Particles.PartitionParticles.Partition, Mask=0, Masked Float=0.8, Passthrough Float=0.6)
    Sprite Size Mode=::NewEnumerator4
    UsePositionOffset=true
    Write Material Random=true
   ) {module=InitializeParticle001}
   &InitialMeshOrientation(Orientation Axis=<0 0 1>, Orientation Vector=<0 0 1>, Rotation=<0 0.25 0>, Use Orientation Vector=true, Use Rotation Vector=true)
   &InitialMeshOrientation(
    Orientation Axis=<0 0 0>
    Orientation Vector=<0 0 0>
    Rotation=&RandomRangeVector(Minimum=<-1 0 0>, Maximum=<1 0 0>)
    Use Orientation Vector=false
    Use Rotation Vector=true
   ) {module=InitialMeshOrientation001}
   /SetVariables(Particles.BigFlareOrient:Quat4f=&OrientMeshToVector(Look At Direction=<0 0 0>, Reference Axis Vector=<1 0 0>)) {isSetParameters=true}
   /SetVariables(Particles.BigFlareDynmat:Vector4f=&MakeVector4(X=&UniformRangedFloat(Minimum=0, Maximum=20), Y=0, Z=0, W=0)) {isSetParameters=true}
   /SetVariables(Particles.VisibilityTag=&SpawnGroup()) {isSetParameters=true}
  ]
  particleUpdateScript [
   &ParticleState()
   &ScaleMeshSize(Scale Factor=&VectorFromFloat(Value=&MaskFloat(
      Mask Source=0
      Mask Parameter=$Particles.PartitionParticles.Partition
      Mask=1
      Masked Float=&FloatFromCurve(
       FloatCurve=DataInterface() {propertyValues {Curve {keys [{arriveTangent=1.9102, interpMode=RCIM_Cubic, leaveTangent=1.9102, tangentMode=RCTM_User, time=0, value=0}, <at 1 1 RCIM_Cubic>]}}}
       Scale Curve=30
      )
      Passthrough Float=0
     )))
   &self:FollowRibbon(
    Alpha=1
    ParticlePerRibbon:NiagaraInt32=$System.ParticlePerRibbon
    RibbonNumber:NiagaraInt32=&Add_Integer(A=$Particles.PartitionParticles.Partition, B=-1)
    RibbonReader=DataInterface() {propertyValues {EmitterBinding {bindingMode=Other, emitterName=NE_EnergyRibbons}}}
   ) {module=FollowRibbon}
   /SetVariables(Particles.Position=&TransformPosition(Position=$System.SampleSkeletalMesh.SampledPosition, Source Space=::World, Destination Space=::Local)) {isSetParameters=true}
   &DynamicMaterialParameters(
    Index 0 Param 1=&FloatFromCurve(FloatCurve=DataInterface() {propertyValues {Curve {keys [{arriveTangent=-2.1015, interpMode=RCIM_Cubic, leaveTangent=-2.1015, tangentMode=RCTM_User, time=0, value=0.5}, <at 1 -0.15>]}}})
    Index 0 Param 2=&FloatFromCurve(FloatCurve=DataInterface() {propertyValues {Curve {keys [{arriveTangent=-0.3464, interpMode=RCIM_Cubic, leaveTangent=-0.3464, tangentMode=RCTM_User, time=0, value=1}, <at 1 0.9>]}}})
    Index 0 Param 3=&FloatFromCurve(FloatCurve=DataInterface() {propertyValues {Curve {keys [
         {interpMode=RCIM_Cubic, tangentMode=RCTM_Break, time=0, value=0}
         {arriveTangent=0.4478, interpMode=RCIM_Cubic, leaveTangent=0.4478, tangentMode=RCTM_User, time=0.1, value=1.05}
         {interpMode=RCIM_Cubic, tangentMode=RCTM_Break, time=1, value=1.2}
        ]}}})
    Module.Param0WriteEnabled=true
    Module.Param1WriteEnabled=true
    Module.Param2WriteEnabled=true
    Module.Param3WriteEnabled=false
   )
   /SetVariables(Particles.BigFlareScale:Vector3f=&Multiply_VectorByFloat(Vector=$Particles.Scale, Float=1)) {isSetParameters=true}
   /SetVariables(Particles.BigFlareColor:LinearColor=&MultiplyLinearColor(
     Linear Color=&MaskLinearColor(
      Mask Parameter=$Particles.PartitionParticles.Partition
      Mask=0
      Masked Linear Color=&ColorFromCurve(ColorCurve=DataInterface() {propertyValues {
         AlphaCurve {keys [<at 0 1 RCIM_Cubic>]}
         BlueCurve {keys [<at 0 10 RCIM_Cubic>,<at 0.6269 1 RCIM_Cubic>]}
         GreenCurve {keys [<at 0 0.0851 RCIM_Cubic>,<at 0.6269 0 RCIM_Cubic>]}
         RedCurve {keys [<at 0 8.9795 RCIM_Cubic>,<at 0.6269 0.5147 RCIM_Cubic>]}
        }})
      Passthrough Linear Color=&ColorFromCurve(ColorCurve=DataInterface() {propertyValues {
         AlphaCurve {keys [<at 0 1 RCIM_Cubic>]}
         BlueCurve {keys [<at 0 0.172 RCIM_Cubic>,<at 0.6269 0.8174 RCIM_Cubic>]}
         GreenCurve {keys [<at 0 0.0007 RCIM_Cubic>,<at 0.6269 0 RCIM_Cubic>]}
         RedCurve {keys [<at 0 10 RCIM_Cubic>,<at 0.6269 1 RCIM_Cubic>]}
        }})
     )
     Scale Factor=&MakeFloat4FromVectorAndFloat(
      "Vector (RGB)"=&VectorFromFloat(Value=1)
      "Float (Alpha)"=&FloatFromCurve(FloatCurve=DataInterface() {propertyValues {Curve {keys [<at 0.4 1 RCIM_Cubic>,<at 1 0>]}}})
     )
    )) {isSetParameters=true}
   &SolveForcesAndVelocity(Clamp Velocity=false, Limit Acceleration=false)
   &FN_Sprite_Attributes(
    Alignment:Vector3f=&RandomVector(Vector Scale=1)
    Color=&Multiply_LinearColorByFloat(
     Linear Color=<rgba 10 0.6764 6.654 1>
     Float=&FloatFromCurve(FloatCurve=DataInterface() {propertyValues {Curve {keys [<at 0 1>,<at 1 0>]}}})
    )
    Create Alignment Attribute=true
    Create Color Attribute=true
    Create SpriteSize Attribute=true
    SpriteSize=&Multiply_Vector2D_ByFloat(
     Float=&FloatFromCurve(FloatCurve=DataInterface() {propertyValues {Curve {keys [<at 0 0>,<at 1 1>]}}})
     Vector2D=<150 220>
    )
    Write to Module Namespace=true
   )
  ]
  rendererClasses [&NiagaraMeshRendererProperties, &NiagaraSpriteRendererProperties]
  renderers [
   {rendererClass=&NiagaraMeshRendererProperties, rendererIndex=0}
   {rendererClass=&NiagaraSpriteRendererProperties, rendererIndex=1}
  ]
 }
 NE_Lightnings {
  emitterUpdateScript [
   &EmitterState()
   &SpawnRate(SpawnRate=&Multiply_Float(A=15, B=$System.FadeOut), Use Spawn Probability=false)
  ]
  particleSpawnScript [
   &InitializeParticle(
    Color=&MultiplyLinearColor(
     Linear Color=&RandomRangeLinearColor(Minimum=<rgba 1 1 1 0.3>, Maximum=<rgba 1.8 1.8 1.8 1>)
     Scale Factor=&MakeVector4_FromFloat(Float=0.8)
    )
    Lifetime=&RandomRangeFloat(Minimum=0.4, Maximum=0.5)
    Mesh Scale Mode=::NewEnumerator3
    Mesh Uniform Scale=&Multiply_Float(A=0.2, B=&Lerp_Float(A=1, B=1.25, Alpha=$System.BuildUp))
    Sprite Size Mode=::NewEnumerator3
    Uniform Sprite Size=20
    UsePositionOffset=true
    Write Material Random=true
   )
   &InitialMeshOrientation(Use Rotation Vector=true)
  ]
  particleUpdateScript [
   &ParticleState()
   &SocketLocation(
    SamplePosOffset=false
    Socket Index=&RandomRangeInt(Minimum=1, Maximum=7)
    Socket Reader=$System.Socket Reader
    Socket Sampling Mode=::NewEnumerator3
    VelocitySpeedLimit=false
   )
   &ScaleMeshSize(Scale Factor=&RandomRangeVector(Minimum=<1 1 2.2>, Maximum=<1.3 1.3 2.2>)) {module=ScaleMeshSize001}
   &DynamicMaterialParameters(
    Index 0 Param 2=&RandomRangeFloat(Minimum=0, Maximum=0)
    Index 0 Param 4=13
    Index 1 Param 1=&FloatFromCurve(FloatCurve=DataInterface() {propertyValues {Curve {keys [<at 0 0.5>,<at 1 1>]}}})
    Index 1 Param 2=2
    Index 1 Param 3=&FloatFromCurve(FloatCurve=DataInterface() {propertyValues {Curve {keys [{interpMode=RCIM_Cubic, tangentMode=RCTM_Break, time=0.5, value=-0.5}, <at 1 0.1>]}}})
    Index 2 Param 1=&Multiply_Float(
     A=&FloatFromCurve(FloatCurve=DataInterface() {propertyValues {Curve {keys [{interpMode=RCIM_Cubic, leaveTangent=8.1085, tangentMode=RCTM_Break, time=0.8568, value=1}, {tangentMode=RCTM_User, time=1, value=0}]}}})
     B=&Divide_Float(
      A=&MakeFloatFromVector(Vector=$Particles.Scale, Channel=::NewEnumerator0)
      B=&RandomRangeFloat(Minimum=10, Maximum=15)
     )
    )
    Index 2 Param 2=&FloatFromCurve(
     FloatCurve=DataInterface() {propertyValues {Curve {keys [
         {interpMode=RCIM_Cubic, tangentMode=RCTM_Break, time=0.7, value=0}
         {arriveTangent=-0.0859, interpMode=RCIM_Cubic, tangentMode=RCTM_Break, time=1, value=-1}
        ]}}}
     Scale Curve=50
    )
    Index 2 Param 3=&Waveform(
     Global Input Position=&RandomRangeFloat(Minimum=0, Maximum=15)
     Global Phase=$Particles.Age
     "Waveform [1]"=::NewEnumerator12
     "[1] Phase Scale"=2
     "[1] Frequency"=1
     "[1] Amplitude Min/Max"=<0 1>
    )
    Index 2 Param 4=25
    Index 3 Param 1=&FloatFromCurve(
     FloatCurve=DataInterface() {propertyValues {Curve {keys [
         {interpMode=RCIM_Cubic, leaveTangent=-0.0025, tangentMode=RCTM_Break, time=0, value=0}
         {arriveTangent=1.8528, interpMode=RCIM_Cubic, tangentMode=RCTM_Break, time=1, value=1}
        ]}}}
     Scale Curve=&RandomRangeFloat(Minimum=0, Maximum=0)
    )
    Index 3 Param 2=&FloatFromCurve(
     FloatCurve=DataInterface() {propertyValues {Curve {keys [
         {interpMode=RCIM_Cubic, leaveTangent=-0.0025, tangentMode=RCTM_Break, time=0, value=0.5}
         {arriveTangent=-0.0316, interpMode=RCIM_Cubic, tangentMode=RCTM_Break, time=1, value=-0.2}
        ]}}}
     Scale Curve=0.1
    )
    Index 3 Param 3=&FloatFromCurve(FloatCurve=DataInterface() {propertyValues {Curve {keys [
         {interpMode=RCIM_Cubic, leaveTangent=1.1034, tangentMode=RCTM_Break, time=0, value=-1}
         {arriveTangent=2.2399, tangentMode=RCTM_User, time=0.7, value=0.125}
        ]}}})
    Index 3 Param 4=&FloatFromCurve(FloatCurve=DataInterface() {propertyValues {Curve {keys [
         {interpMode=RCIM_Cubic, leaveTangent=3.0601, tangentMode=RCTM_Break, time=0.5, value=-0.5}
         {arriveTangent=3.2053, tangentMode=RCTM_User, time=1, value=1}
        ]}}})
    Module.Index0Param0Write=true
    Module.Index0Param1Write=true
    Module.Index0Param2Write=true
    Module.Index0Param3Write=true
    Module.Index1Param0Write=true
    Module.Index1Param1Write=true
    Module.Index1Param2Write=true
    Module.Index1Param3Write=true
    Module.Index2Param0Write=true
    Module.Index2Param1Write=true
    Module.Index2Param2Write=true
    Module.Index2Param3Write=true
    Module.Index3Param0Write=true
    Module.Index3Param1Write=true
    Module.Index3Param2Write=true
    Module.Index3Param3Write=true
    Write Parameter Index 1=true
    Write Parameter Index 2=true
    Write Parameter Index 3=true
   )
   &SolveForcesAndVelocity(Clamp Velocity=false, Limit Acceleration=false)
   &Color(
    Color=&MultiplyLinearColor(
     Linear Color=&ColorFromCurve(ColorCurve=DataInterface() {propertyValues {
        AlphaCurve {keys [<at 0 1 RCIM_Cubic>,<at 0.7884 1>,<at 1 0 RCIM_Cubic>]}
        BlueCurve {keys [<at 0.2225 5>, {interpMode=RCIM_Cubic, leaveTangent=-19.7604, time=0.4906, value=1.5}]}
        GreenCurve {keys [<at 0.2225 5>, {interpMode=RCIM_Cubic, leaveTangent=-19.7604, time=0.4906, value=1.5}]}
        RedCurve {keys [<at 0.2225 5>, {interpMode=RCIM_Cubic, leaveTangent=-19.7604, time=0.4906, value=1.5}]}
       }})
     Scale Factor=&MakeVector4_FromLinearColor(LinearColor=<rgba 0.6296 0.05 1 1>)
    )
    Scale Alpha=&FloatFromCurve(FloatCurve=DataInterface() {propertyValues {Curve {keys [
         {interpMode=RCIM_Cubic, leaveTangent=2.5294, tangentMode=RCTM_Break, time=0, value=0}
         {arriveTangent=-0.2166, interpMode=RCIM_Cubic, tangentMode=RCTM_Break, time=0.4, value=1}
        ]}}})
   )
   &CameraOffset(Camera Offset Amount=10)
  ]
  rendererClasses [&NiagaraMeshRendererProperties]
  renderers [
   {rendererClass=&NiagaraMeshRendererProperties, rendererIndex=0}
   {rendererClass=&NiagaraMeshRendererProperties, rendererIndex=1}
  ]
 }
 NE_Sparks {
  bEnabled=false
  emitterUpdateScript [
   &EmitterState(Inactive Response=::NewEnumerator1, Life Cycle Mode=::NewEnumerator1, Loop Duration=1, UseLoopDelay=false)
   &SpawnRate(SpawnRate=30, Use Spawn Probability=false)
  ]
  particleSpawnScript [&InitializeParticle(
    Color=<rgba 9.6162 0 10 1>
    Lifetime=&RandomRangeFloat(Minimum=1, Maximum=1.5)
    Sprite Size=&Multiply_Vector2D_ByFloat(Float=&RandomRangeFloat(Minimum=1, Maximum=2), Vector2D=<3 12>)
    Sprite Size Mode=::NewEnumerator4
    UsePositionOffset=false
    Write Material Random=true
   )]
  particleUpdateScript [
   &ParticleState()
   &self:FollowRibbon(
    Alpha:NiagaraFloat=&FloatFromCurve(FloatCurve=DataInterface() {propertyValues {Curve {keys [<at 0 0.95>,<at 1 0>]}}})
    ParticlePerRibbon:NiagaraInt32=$System.ParticlePerRibbon
    RibbonNumber:NiagaraInt32=&RandomRangeInt(Minimum=0, Maximum=2)
    RibbonReader=DataInterface() {propertyValues {EmitterBinding {bindingMode=Other, emitterName=NE_EnergyRibbons}}}
   ) {module=FollowRibbon}
   /SetVariables(Particles.Position=&Add_Vector(
     A=Linked() {linkedVariable {name=Output.FollowRibbon.RibbonPos, type=NiagaraPosition}}
     B=&RandomVector(Vector Scale=&FloatFromCurve(
       FloatCurve=DataInterface() {propertyValues {Curve {keys [
           {interpMode=RCIM_Cubic, tangentMode=RCTM_User, time=0, value=1}
           {arriveTangent=-3.4279, interpMode=RCIM_Cubic, leaveTangent=-3.4279, tangentMode=RCTM_User, time=1, value=0}
          ]}}}
       Scale Curve=60
      ))
    )) {isSetParameters=true}
   &SpriteFacingAndAlignment(Sprite Alignment=$Output.FollowRibbon.DirectionVector, WriteAlignment=true, WriteFacing=false)
   &Color(Color=&MultiplyLinearColor(
     Linear Color=&ColorFromCurve(ColorCurve=DataInterface() {propertyValues {
        AlphaCurve {keys [<at 0 1 RCIM_Cubic>]}
        BlueCurve {keys [<at 0.1263 0.0221 RCIM_Cubic>,<at 1 6.3633 RCIM_Cubic>]}
        GreenCurve {keys [<at 0.1263 0 RCIM_Cubic>,<at 1 0 RCIM_Cubic>]}
        RedCurve {keys [<at 0.1263 10 RCIM_Cubic>,<at 1 8 RCIM_Cubic>]}
       }})
     Scale Factor=&MakeFloat4FromVectorAndFloat("Vector (RGB)"=<1 1 1>, "Float (Alpha)"=$System.FadeOut)
    ))
   &FN_Sprite_Attributes(
    Color=&MultiplyLinearColor(Linear Color=$Particles.Color, Scale Factor=<1 1 1 0.05>)
    Create Color Attribute=true
    Create SpriteSize Attribute=true
    SpriteSize=&Multiply_Vector2D(A=$Particles.SpriteSize, B=<10 5>)
    Write to Module Namespace=true
   )
  ]
  rendererClasses [&NiagaraSpriteRendererProperties]
  renderers [
   {rendererClass=&NiagaraSpriteRendererProperties}
   {rendererClass=&NiagaraSpriteRendererProperties, rendererIndex=1}
  ]
 }
 Smoke {
  bEnabled=false
  emitterUpdateScript [
   &EmitterState(Life Cycle Mode=::NewEnumerator1, Loop Duration=1, UseLoopDelay=false)
   &SpawnRate(SpawnRate=15, Use Spawn Probability=false)
   &SpawnRate(SpawnGroup=2, SpawnRate=30, Use Spawn Probability=false) {module=SpawnRate002}
  ]
  particleSpawnScript [
   &PartitionParticles(Number Of Partitions=$System.RibbonNumber, Partition Mode:ENiagara_PartitionMode=::NewEnumerator5)
   &InitializeParticle(
    Color=&SelectLinearColorFromArray(
     Array Sampling Mode=::NewEnumerator1
     Direct Array Index=&SpawnGroup()
     Color Selection Array=DataInterface() {propertyValues {ColorData [<rgba 0.0487 0.0057 0.1 2>, <rgba 0.8281 0 0.7022 2>, <rgba 0.1 0.0013 0.0009 2>]}}
     Random Seed=?
    )
    Lifetime=&SelectFloatFromArray(
     Direct Array Index=&SpawnGroup()
     Float Selection Array=DataInterface() {propertyValues {FloatData [0.7,1.5,2.5]}}
    )
    Position Offset Coordinate Space=::Local
    Sprite Rotation Mode=::NewEnumerator2
    Sprite Size Mode=::NewEnumerator3
    Sprite UV Mode=::NewEnumerator3
    Uniform Sprite Size=&Multiply_Float(
     A=&RandomRangeFloat(Minimum=50, Maximum=80)
     B=&SelectFloatFromArray(Direct Array Index=&SpawnGroup(), Float Selection Array=DataInterface() {propertyValues {FloatData [1.2,1,2]}})
    )
    UsePositionOffset=true
    Write Material Random=true
   )
   &SubUVAnimation(
    End Frame Range Override=4
    Evaluation Type=::NewEnumerator0
    "Sprite Renderer "=DataInterface() {propertyValues {SpriteRenderer=&self:Smoke_0.NiagaraSpriteRendererProperties_1}}
    SubUV Animation Mode:ENiagara_SubUVLookupModeV2=::NewEnumerator2
    UseEndFrame=true
    UseStartFrame=true
   )
   &CameraOffset(Camera Offset Amount=&MaskFloatBySpawnGroup(Masked Float=0, Passthrough Float=10))
  ]
  particleUpdateScript [
   &ParticleState()
   &SpriteRotationRate(Rotation Rate=&RandomRangeFloat(Minimum=-10, Maximum=10))
   &self:FollowRibbon(
    Alpha:NiagaraFloat=&Lerp_Float(
     A=1
     B=0.3
     Alpha=&FloatFromCurve(FloatCurve=DataInterface() {propertyValues {Curve {keys [<at 0 0>,<at 1 1>]}}})
    )
    ParticlePerRibbon:NiagaraInt32=$System.ParticlePerRibbon
    RibbonNumber:NiagaraInt32=$Particles.PartitionParticles.Partition
    RibbonReader=DataInterface() {propertyValues {EmitterBinding {bindingMode=Other, emitterName=NE_EnergyRibbons}}}
   ) {module=FollowRibbon}
   &Drag(Use Linear Drag=true, Use Rotational Drag=false)
   &SolveForcesAndVelocity(Clamp Velocity=false, Limit Acceleration=false)
   &ScaleSpriteSize(
    Scale Sprite Size Mode=::NewEnumerator0
    Uniform Scale Factor=&Multiply_Float(
     A=&MaskFloat(
      Mask Parameter=$Particles.PartitionParticles.Partition
      Mask=0
      Masked Float=&FloatFromCurve(FloatCurve=DataInterface() {propertyValues {Curve {keys [
           {arriveTangent=0.4033, interpMode=RCIM_Cubic, leaveTangent=0.4033, tangentMode=RCTM_User, time=0, value=1}
           {arriveTangent=0.0195, interpMode=RCIM_Cubic, leaveTangent=0.0195, tangentMode=RCTM_User, time=1, value=0}
          ]}}})
      Passthrough Float=&FloatFromCurve(
       FloatCurve=DataInterface() {propertyValues {Curve {keys [
           {arriveTangent=4.1785, interpMode=RCIM_Cubic, leaveTangent=4.1785, tangentMode=RCTM_User, time=0, value=0}
           {arriveTangent=0.0224, interpMode=RCIM_Cubic, leaveTangent=0.0224, tangentMode=RCTM_User, time=0.1, value=1}
           {arriveTangent=-1.4679, interpMode=RCIM_Cubic, leaveTangent=-1.4679, tangentMode=RCTM_User, time=1, value=0.5}
          ]}}}
       Scale Curve=0.7
      )
     )
     B=&MaskFloatBySpawnGroup(Masked Float=1.75, Passthrough Float=1)
    )
   )
   /SetVariables(Particles.Position=&MergeHelper_VectorSwitch(
     0=&TransformPosition(Position=$System.SampleSkeletalMesh.SampledPosition, Source Space=::World, Destination Space=::Local)
     1=&TransformPosition(Position=$System.SampleSkeletalMesh.SampledPosition, Source Space=::World, Destination Space=::Local)
     2=$Output.FollowRibbon.RibbonPos
     3=?
     Index=&SpawnGroup()
     Num=::NewEnumerator1
    )) {isSetParameters=true}
   &DynamicMaterialParameters(
    Index 0 Param 1=&RandomRangeFloat(Minimum=0, Maximum=111)
    Index 0 Param 4=0.2
    Module.Param0WriteEnabled=true
    Module.Param1WriteEnabled=false
    Module.Param2WriteEnabled=false
    Module.Param3WriteEnabled=false
   ) {module=DynamicMatParams}
   &ScaleColor(
    Scale Alpha=&Multiply_Float(
     A=1
     B=&FloatFromCurve(
      FloatCurve=DataInterface() {propertyValues {Curve {keys [
          {interpMode=RCIM_Cubic, tangentMode=RCTM_User, time=0, value=0}
          {interpMode=RCIM_Cubic, tangentMode=RCTM_User, time=0.3, value=1}
          {interpMode=RCIM_Cubic, tangentMode=RCTM_User, time=0.5, value=1}
          {tangentMode=RCTM_User, time=1, value=0}
         ]}}}
      Scale Curve=$System.FadeOut
     )
    )
    Scale RGB=&VectorFromFloat(Value=&FNTimeOfDayScalar(Night Scale=0.6, Morning Scale=0.8, Evening Scale=0.8))
    ScaleA=true
    ScaleRGB=true
   )
   &KillParticles(Kill Particles=/Niagara/DynamicInputs/Helpers/MaskBoolBySpawnGroup(
     Masked Bool:NiagaraBool=false
     Passthrough Bool:NiagaraBool=&Boolean_NotOperation(Bool=/Game/Effects/Niagara/DynamicInputs/State/FN_IsExecutionStateActive())
     Spawn Group Mask:NiagaraInt32=0
    ))
  ]
  rendererClasses [&NiagaraSpriteRendererProperties]
  renderers [{rendererClass=&NiagaraSpriteRendererProperties}]
 }
}
summary {
 systemName=NS_PistolCanary_MM_DrainSource
 emitters [
  {
   emitterName=NE_BaseGlow
   rendererClasses [
    &NiagaraSpriteRendererProperties
    &NiagaraLightRendererProperties
    &NiagaraMeshRendererProperties
   ]
  }
  {bEnabled=false, emitterName=Smoke, rendererClasses [&NiagaraSpriteRendererProperties]}
  {
   bEnabled=false
   emitterName=NE_Flare
   rendererClasses [&NiagaraMeshRendererProperties, &NiagaraSpriteRendererProperties]
  }
  {bEnabled=false, emitterName=NE_EnergyRibbons, rendererClasses [&NiagaraRibbonRendererProperties]}
  {bEnabled=false, emitterName=NE_Sparks, rendererClasses [&NiagaraSpriteRendererProperties]}
  {emitterName=NE_Lightnings, rendererClasses [&NiagaraMeshRendererProperties]}
 ]
 userVariables [User.FadeOut {type=NiagaraFloat, defaultValue=1, description=""}]
}
