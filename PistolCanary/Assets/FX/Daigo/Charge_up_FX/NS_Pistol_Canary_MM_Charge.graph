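# Niagara charge-up effect for the Pistol Canary, layering hand energy, glows, bursts and leaking particles over the charge.
# System-level timings drive the emitters: five burst times trigger flare and spark bursts, and a hand-charge curve sizes a glowing energy ball offset by a user hand position.
# User FlareControl scales tendrils, a spinning energy mesh, glow and leak spawning, which mostly build with loop age.
# Some sizes switch by platform quality, and some alphas by time of day.
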NiagaraSystem /PistolCanary/Assets/FX/Daigo/Charge_up_FX/NS_Pistol_Canary_MM_Charge
dependencies {
 usedDataInterfaces [
  &NiagaraDataInterfaceCurve
  &NiagaraDataInterfaceCamera
  &NiagaraDataInterfaceColorCurve
  &NiagaraDataInterfacePlatformSet
 ]
 usedDynamicInputs []
 usedModules [
  &self:SystemSpawnScript.SystemScriptSource.SystemScriptGraph.NiagaraNodeAssignment_0.SetVariables_#1
  &self:SystemSpawnScript.SystemScriptSource.SystemScriptGraph.NiagaraNodeAssignment_1.SetVariables_#2
  &self:SystemSpawnScript.SystemScriptSource.SystemScriptGraph.NiagaraNodeAssignment_2.SetVariables_#3
  &self:SystemSpawnScript.SystemScriptSource.SystemScriptGraph.NiagaraNodeAssignment_3.SetVariables_#4
  &SystemState
 ]
}
emitters {
 Air_Tendrils_Large {
  emitterUpdateScript [
   &EmitterState()
   &SpawnBurst_Instantaneous(Spawn Count=5, Use Spawn Probability=false, UseLoopCountLimit=false) {enabled=false}
   &SpawnRate(
    SpawnRate=&FloatFromCurve(
     FloatCurve=DataInterface() {propertyValues {Curve {keys [<at 0 0>,<at 0.8 20>,<at 0.95 0>]}}}
     CurveIndex=$Emitter.NormalizedLoopAge
     Scale Curve=$User.FlareControl
    )
    Use Spawn Probability=false
   )
   /SetVariables() {isSetParameters=true}
  ]
  particleSpawnScript [
   &InitializeParticle(
    Color=<rgba 0.8371 0.0435 2 0.5>
    Lifetime Max=1
    Lifetime Min=0.6
    Lifetime Mode=::NewEnumerator1
    Mesh Scale Max=<3 3 3>
    Mesh Scale Min=<2 2 2>
    Mesh Scale Mode=::NewEnumerator2
    Position Offset=&RandomRangeVector(Minimum=<0 0 0>, Maximum=<0 0 50>)
    UsePositionOffset=true
    Write Material Random=true
   ) {module=InitializeParticle001}
   &InitialMeshOrientation(
    Rotation=&RandomRangeVector(Minimum=<-1 -1 -1>, Maximum=<1 1 1>)
    Use Orientation Vector=true
    Use Rotation Vector=true
   ) {enabled=false}
   &InitialMeshOrientation(Use Rotation Vector=false) {module=InitialMeshOrientation001}
  ]
  particleUpdateScript [
   &ParticleState()
   &SolveForcesAndVelocity(Clamp Velocity=false, Limit Acceleration=false)
   &ScaleColor(
    Scale Alpha=&Multiply_Float(
     A=&FloatFromCurve(FloatCurve=DataInterface() {propertyValues {Curve {keys [
          {interpMode=RCIM_Cubic, tangentMode=RCTM_User, time=0, value=0}
          {interpMode=RCIM_Cubic, tangentMode=RCTM_User, time=0.5, value=1}
          {tangentMode=RCTM_User, time=1, value=0}
         ]}}})
     B=1
    )
    Scale RGB=&Multiply_VectorByFloat(
     Vector=&MakeVectorFromLinearColorRGB(Color=&ColorFromCurve(ColorCurve=DataInterface() {propertyValues {
         AlphaCurve {keys []}
         BlueCurve {keys [<at 0.1479 0.4154 RCIM_Cubic>,<at 0.9333 1.9264>]}
         GreenCurve {keys [<at 0.1479 0.152 RCIM_Cubic>,<at 0.9333 2>]}
         RedCurve {keys [<at 0.1479 1 RCIM_Cubic>,<at 0.9333 1.8638>]}
        }}))
     Float=&FloatFromCurve(
      FloatCurve=DataInterface() {propertyValues {Curve {keys [<at 0 1 RCIM_Cubic>, {arriveTangent=7.7988, tangentMode=RCTM_User, time=1, value=5}]}}}
      CurveIndex=$System.NormalizedLoopAge
     )
    )
    ScaleA=true
    ScaleRGB=true
   )
   &CameraOffset(Camera Offset Amount=20)
   &ScaleMeshSize(Scale Factor=&Multiply_VectorByFloat(
     Vector=&Multiply_VectorByFloat(
      Vector=&Multiply_Vector(A=<1.3 1 0.5>, B=<1.3 2 1>)
      Float=&Multiply_Float(
       A=&FloatFromCurve(FloatCurve=DataInterface() {propertyValues {Curve {keys [
            {interpMode=RCIM_Cubic, tangentMode=RCTM_User, time=0, value=0.5}
            {arriveTangent=-0.752, interpMode=RCIM_Cubic, leaveTangent=-0.752, tangentMode=RCTM_User, time=1, value=0.3}
           ]}}})
       B=2
      )
     )
     Float=$System.FX Scale
    ))
   &DynamicMaterialParameters(
    Index 0 Param 3=&Multiply_Float(
     A=$Engine.Time
     B=&Add_Float(
      A=&RandomRangeFloat(Minimum=5, Maximum=6)
      B=&RandomRangeFloat(Minimum=0, Maximum=1)
     )
    )
    Index 2 Param 1=&Multiply_Float(A=&RandomRangeFloat(Minimum=70, Maximum=30), B=$System.FX Scale)
    Index 2 Param 2=&FloatFromCurve(FloatCurve=DataInterface() {propertyValues {Curve {keys [
         {arriveTangent=0.1852, interpMode=RCIM_Cubic, leaveTangent=0.1852, tangentMode=RCTM_User, time=0, value=0.6}
         {arriveTangent=-0.7912, interpMode=RCIM_Cubic, leaveTangent=-2.4631, tangentMode=RCTM_Break, time=0.5118, value=0.833}
         {arriveTangent=-1.0004, interpMode=RCIM_Cubic, leaveTangent=-0.6901, tangentMode=RCTM_Break, time=1, value=0}
        ]}}})
    Module.Index0Param0Write=false
    Module.Index0Param1Write=false
    Module.Index0Param2Write=true
    Module.Index0Param3Write=false
    Module.Index2Param0Write=true
    Module.Index2Param1Write=true
    Module.Index2Param2Write=true
    Module.Index2Param3Write=true
    Write Parameter Index 2=true
   ) {module=DynamicMaterialParameters001}
   /SetVariables() {isSetParameters=true}
  ]
  rendererClasses [&NiagaraMeshRendererProperties]
  renderers [{rendererClass=&NiagaraMeshRendererProperties, rendererIndex=0}]
 }
 Burst {
  simTarget=GPUComputeSim
  emitterUpdateScript [
   &EmitterState(Life Cycle Mode=::NewEnumerator1, Loop Behavior=::NewEnumerator1, Loop Duration=20, UseLoopDelay=false)
   &SpawnBurst_Instantaneous(Spawn Time=$System.Burst1, Use Spawn Probability=false, UseLoopCountLimit=false)
   &SpawnBurst_Instantaneous(Spawn Time=$System.Burst2, Use Spawn Probability=false, UseLoopCountLimit=false) {module=SpawnBurst_Instantaneous001}
   &SpawnBurst_Instantaneous(Spawn Time=$System.Burst3, Use Spawn Probability=false, UseLoopCountLimit=false) {module=SpawnBurst_Instantaneous002}
   &SpawnBurst_Instantaneous(Spawn Time=$System.Burst4, Use Spawn Probability=false, UseLoopCountLimit=false) {module=SpawnBurst_Instantaneous003}
   &SpawnBurst_Instantaneous(Spawn Time=$System.Burst5, Use Spawn Probability=false, UseLoopCountLimit=false) {module=SpawnBurst_Instantaneous004}
  ]
  particleSpawnScript [
   &InitializeParticle(
    Color=<rgba 1.5 0.5927 1.2985 0.5>
    Lifetime=0.265
    Sprite Size Mode=::NewEnumerator3
    Uniform Sprite Size=2000
    UsePositionOffset=false
    Write Material Random=true
   )
   &CameraOffset(Camera Offset Amount=1200) {module=CameraOffset001}
   /SetVariables(
    Particles.Animorphic Color:LinearColor=<rgba 5 0.1507 0.5548 0.05>
    Particles.Animorphic Rotation=90
    Particles.Animorphic Size:Vector2f=<1800 150>
    Particles.Animorphic Vector:Vector3f=<0 0 1>
    Particles.BigGlowSize:Vector2f=<4000 2500>
    Particles.Soft Glow Size:Vector2f=<2000 500>
   ) {isSetParameters=true, module=SetVariables_#5}
  ]
  particleUpdateScript [
   &ParticleState()
   &ScaleColor(
    Scale Alpha=&FloatFromCurve(FloatCurve=DataInterface() {propertyValues {Curve {keys [
         {interpMode=RCIM_Cubic, tangentMode=RCTM_User, time=0, value=0}
         {interpMode=RCIM_Cubic, tangentMode=RCTM_User, time=0.5, value=1}
         {tangentMode=RCTM_User, time=1, value=0}
        ]}}})
    ScaleA=true
    ScaleRGB=true
   )
   /SetVariables(Particles.Soft Glow Color:LinearColor=&MakeLinearColorFromVectorAndFloat(
     "Vector (RGB)"=&MakeVectorFromLinearColorRGB(Color=<rgba 5 1.541 0.1027 1>)
     "Float (Alpha)"=&FloatFromCurve(
      FloatCurve=DataInterface() {propertyValues {Curve {keys [
          {interpMode=RCIM_Cubic, tangentMode=RCTM_User, time=0, value=0}
          {interpMode=RCIM_Cubic, tangentMode=RCTM_User, time=0.5, value=1}
          {tangentMode=RCTM_User, time=1, value=0}
         ]}}}
      Scale Curve=&Multiply_Float(
       A=0.005
       B=&FloatFromCurve(
        FloatCurve=DataInterface() {propertyValues {Curve {keys [{interpMode=RCIM_Cubic, leaveTangent=-3.7201, tangentMode=RCTM_Break, time=0, value=1}, <at 1 0>]}}}
        CurveIndex=$System.TimeForDis
       )
      )
     )
    )) {isSetParameters=true, module=SetVariables_#6}
   /SetVariables(
    Particles.Corona Color:LinearColor=&MakeLinearColorFromVectorAndFloat(
     "Vector (RGB)"=&MakeVectorFromLinearColorRGB(Color=<rgba 1 0.0889 0.1786 1>)
     "Float (Alpha)"=&FloatFromCurve(
      FloatCurve=DataInterface() {propertyValues {Curve {keys [
          <at 0 0 RCIM_Cubic>
          {arriveTangent=2.5, interpMode=RCIM_Cubic, leaveTangent=2.5, time=0.165, value=0.65}
          {arriveTangent=0.1786, interpMode=RCIM_Cubic, leaveTangent=0.1786, time=0.4, value=1}
          {arriveTangent=-1.6667, interpMode=RCIM_Cubic, leaveTangent=-1.6667, time=0.725, value=0.75}
          <at 1 0 RCIM_Cubic>
         ]}}}
      Scale Curve=&Multiply_Float(
       A=0.5
       B=&FloatFromCurve(
        FloatCurve=DataInterface() {propertyValues {Curve {keys [{interpMode=RCIM_Cubic, leaveTangent=-3.7201, tangentMode=RCTM_Break, time=0, value=1}, <at 1 0>]}}}
        CurveIndex=$System.TimeForDis
       )
      )
     )
    )
    Particles.Corona Size:Vector2f=&Vector2DFromFloat(Value=&FloatFromCurve(
      FloatCurve=DataInterface() {propertyValues {Curve {keys [<at 0 1>,<at 0.5 0.7>,<at 1 1>]}}}
      Scale Curve=&Waveform(Global Amplitude Scale=2000, "Waveform [1]"=::NewEnumerator12, "[1] Phase Scale"=50, "[1] Frequency"=1, "[1] Amplitude Min/Max"=<0.85 1>)
     ))
   ) {isSetParameters=true, module=SetVariables_#7}
   &DynamicMaterialParameters(
    Index 3 Param 1=&FloatFromCurve(FloatCurve=DataInterface() {propertyValues {Curve {keys [<at 0 0.7 RCIM_Cubic>,<at 0.5 0.25 RCIM_Cubic>,<at 1 0.7 RCIM_Cubic>]}}})
    Index 3 Param 2=1
    Index 3 Param 3=&RandomRangeFloat(Minimum=0.25, Maximum=0.35)
    Index 3 Param 4=1
    Module.Param0WriteEnabled3=true
    Module.Param1WriteEnabled3=true
    Module.Param2WriteEnabled3=true
    Module.Param3WriteEnabled3=true
    Write Parameter Index 0=false
    Write Parameter Index 3=true
   )
   /SetVariables() {isSetParameters=true}
   &ScaleColor(
    Scale Alpha=&FloatFromCurve(
     FloatCurve=DataInterface() {propertyValues {Curve {keys [{interpMode=RCIM_Cubic, leaveTangent=-3.7201, tangentMode=RCTM_Break, time=0, value=1}, <at 1 0>]}}}
     CurveIndex=$System.TimeForDis
    )
    ScaleA=true
    ScaleRGB=true
   ) {module=ScaleColor001}
   /SetVariables(Particles.Animorphic Color:LinearColor=&MultiplyLinearColor(
     Linear Color=$Particles.Initial.Animorphic Color
     Scale Factor=&MakeVector4FromVectorAndFloat(
      "Vector (RGB)"=<1 1 1>
      "Float (Alpha)"=&FloatFromCurve(
       FloatCurve=DataInterface() {propertyValues {Curve {keys [{interpMode=RCIM_Cubic, leaveTangent=-3.7201, tangentMode=RCTM_Break, time=0, value=1}, <at 1 0>]}}}
       CurveIndex=$System.TimeForDis
      )
     )
    )) {isSetParameters=true, module=SetVariables_#8}
   /SetVariables(Particles.BigGlowColor:LinearColor=&MultiplyLinearColor(
     Linear Color=$Particles.Animorphic Color
     Scale Factor=&MakeVector4FromVectorAndFloat(
      "Vector (RGB)"=<0.5 0.5 0.5>
      "Float (Alpha)"=&FloatFromCurve(FloatCurve=DataInterface() {propertyValues {Curve {
          defaultValue=1
          keys [
           {arriveTangent=9.0725, interpMode=RCIM_Cubic, leaveTangent=9.0725, tangentMode=RCTM_User, time=0, value=0}
           {arriveTangent=-2.5249, interpMode=RCIM_Cubic, leaveTangent=-2.5249, tangentMode=RCTM_User, time=0.1408, value=0.9847}
           {arriveTangent=-0.0405, interpMode=RCIM_Cubic, leaveTangent=-0.0405, tangentMode=RCTM_User, time=1, value=0}
          ]
         }}})
     )
    )) {isSetParameters=true}
  ]
  rendererClasses [&NiagaraSpriteRendererProperties]
  renderers [
   {rendererClass=&NiagaraSpriteRendererProperties}
   {rendererClass=&NiagaraSpriteRendererProperties, rendererIndex=1}
   {rendererClass=&NiagaraSpriteRendererProperties, rendererIndex=2}
   {rendererClass=&NiagaraSpriteRendererProperties, rendererIndex=3}
  ]
 }
 BurstSparks {
  simTarget=GPUComputeSim
  emitterSpawnScript [/SetVariables(Emitter.BurstCount=65) {isSetParameters=true}]
  emitterUpdateScript [
   &EmitterState(Life Cycle Mode=::NewEnumerator1, Loop Behavior=::NewEnumerator1, Loop Duration=20, UseLoopDelay=false)
   &SpawnBurst_Instantaneous(Spawn Count=$Emitter.BurstCount, Spawn Time=$System.Burst1, Use Spawn Probability=false, UseLoopCountLimit=false)
   &SpawnBurst_Instantaneous(Spawn Count=$Emitter.BurstCount, Spawn Time=$System.Burst2, Use Spawn Probability=false, UseLoopCountLimit=false) {module=SpawnBurst_Instantaneous001}
   &SpawnBurst_Instantaneous(Spawn Count=$Emitter.BurstCount, Spawn Time=$System.Burst3, Use Spawn Probability=false, UseLoopCountLimit=false) {module=SpawnBurst_Instantaneous002}
   &SpawnBurst_Instantaneous(Spawn Count=$Emitter.BurstCount, Spawn Time=$System.Burst4, Use Spawn Probability=false, UseLoopCountLimit=false) {module=SpawnBurst_Instantaneous003}
   &SpawnBurst_Instantaneous(Spawn Count=$Emitter.BurstCount, Spawn Time=$System.Burst5, Use Spawn Probability=false, UseLoopCountLimit=false) {module=SpawnBurst_Instantaneous004}
  ]
  particleSpawnScript [
   &InitializeParticle(
    Color=<rgba 275 65.3025 230.0222 1>
    Lifetime Max=1.5
    Lifetime Min=0.5
    Lifetime Mode=::NewEnumerator1
    Sprite Size=&Multiply_Vector2D_ByFloat(Float=&RandomRangeFloat(Minimum=0.1, Maximum=1), Vector2D=<50 100>)
    Sprite Size Mode=::NewEnumerator4
    UsePositionOffset=false
    Write Material Random=true
   )
   &ShapeLocation(
    MaskSpawnGroup=false
    Override Seed=false
    Sphere Radius=&RandomRangeFloat(Minimum=5, Maximum=100)
    UseOwnerScale=true
   )
   &AddVelocity(
    Module.Use Falloff Distance=false
    Override Seed=false
    Velocity Mode=::NewEnumerator1
    Velocity Speed=&RandomRangeFloat(Minimum=50, Maximum=2000)
   )
  ]
  particleUpdateScript [
   &ParticleState()
   &Drag(Use Linear Drag=true, Use Rotational Drag=true)
   &GravityForce()
   &SolveForcesAndVelocity(Clamp Velocity=false, Limit Acceleration=false)
   &ScaleSpriteSizeBySpeed(Max Scale Factor=<1 2>, Min Scale Factor=<0.5 1>, Velocity Threshold=12000)
   &ScaleSpriteSize(Uniform Curve Sprite Scale=DataInterface() {propertyValues {Curve {
       defaultValue=0
       keys [{interpMode=RCIM_Cubic, tangentMode=RCTM_User, time=0, value=1}, {interpMode=RCIM_Cubic, tangentMode=RCTM_User, time=1, value=0}]
      }}})
  ]
  rendererClasses [&NiagaraSpriteRendererProperties]
  renderers [{rendererClass=&NiagaraSpriteRendererProperties}]
 }
 EnergyBall {
  emitterUpdateScript [
   &EmitterState(
    Life Cycle Mode=::NewEnumerator1
    Loop Behavior=::NewEnumerator1
    Loop Delay=$System.HandchargeStartTime
    Loop Duration=$System.HandchargeLoopTime
    UseLoopDelay=true
   )
   &SpawnBurst_Instantaneous(Use Spawn Probability=false, UseLoopCountLimit=false)
   &SpawnRate(SpawnRate=25, Use Spawn Probability=false)
   /SetVariables(Emitter.float:NiagaraFloat=&FloatFromCurve(FloatCurve=$System.HandchargeCurve, CurveIndex=$Emitter.NormalizedLoopAge)) {isSetParameters=true}
  ]
  particleSpawnScript [
   &InitializeParticle(
    Color=$User.PrimaryColor
    Lifetime Max=0.175
    Lifetime Min=0.075
    Lifetime Mode=::NewEnumerator1
    Position Offset=<0 50 150>
    Position Offset Coordinate Space=::World
    Sprite Rotation Mode=::NewEnumerator2
    Sprite Size Mode=::NewEnumerator3
    Uniform Sprite Size=&Multiply_Float(
     A=&RandomRangeFloat(Minimum=400, Maximum=700)
     B=&FloatFromCurve(FloatCurve=$System.HandchargeCurve, CurveIndex=$Emitter.NormalizedLoopAge)
    )
    UsePositionOffset=true
    Write Material Random=true
   )
   /SetVariables(
    Particles.AnimRotation2=90
    Particles.AnimSize2:Vector2f=&Multiply_Vector2D_ByFloat(
     Float=&FloatFromCurve(FloatCurve=$System.HandchargeCurve, CurveIndex=$Emitter.NormalizedLoopAge)
     Vector2D=<1500 75>
    )
    Particles.AnimVector2:Vector3f=<0 0 1>
    Particles.Animorphic Color:LinearColor=<rgba 20 0.6026 2.219 0.05>
    Particles.Animorphic Size:Vector2f=&Multiply_Vector2D_ByFloat(
     Float=&FloatFromCurve(FloatCurve=$System.HandchargeCurve, CurveIndex=$Emitter.NormalizedLoopAge)
     Vector2D=<2000 20>
    )
    Particles.Soft Glow Size:Vector2f=&Multiply_Vector2D_ByFloat(
     Float=&FloatFromCurve(FloatCurve=$System.HandchargeCurve, CurveIndex=$Emitter.NormalizedLoopAge)
     Vector2D=<3289.3108 195.8096>
    )
   ) {isSetParameters=true, module=SetVariables_#5}
  ]
  particleUpdateScript [
   &ParticleState()
   &ScaleColor(
    Scale Alpha=&FloatFromCurve(FloatCurve=DataInterface() {propertyValues {Curve {keys [
         {interpMode=RCIM_Cubic, tangentMode=RCTM_User, time=0, value=0}
         {interpMode=RCIM_Cubic, tangentMode=RCTM_User, time=0.5, value=1}
         {tangentMode=RCTM_User, time=1, value=0}
        ]}}})
    Scale RGB=<100 100 100>
    ScaleA=true
    ScaleRGB=true
   )
   &CameraOffset(Camera Offset Amount=100)
   /SetVariables(
    Particles.SGlowColor2:LinearColor=&MakeLinearColorFromVectorAndFloat(
     "Vector (RGB)"=&MakeVectorFromLinearColorRGB(Color=<rgba 10 3.0819 0.2054 1>)
     "Float (Alpha)"=&FloatFromCurve(
      FloatCurve=DataInterface() {propertyValues {Curve {keys [
          {interpMode=RCIM_Cubic, tangentMode=RCTM_User, time=0, value=0}
          {interpMode=RCIM_Cubic, tangentMode=RCTM_User, time=0.5, value=1}
          {tangentMode=RCTM_User, time=1, value=0}
         ]}}}
      Scale Curve=0.005
     )
    )
    Particles.Soft Glow Color:LinearColor=&MakeLinearColorFromVectorAndFloat(
     "Vector (RGB)"=&MakeVectorFromLinearColorRGB(Color=<rgba 10 0 0.8333 1>)
     "Float (Alpha)"=&FloatFromCurve(
      FloatCurve=DataInterface() {propertyValues {Curve {keys [
          {interpMode=RCIM_Cubic, tangentMode=RCTM_User, time=0, value=0}
          {interpMode=RCIM_Cubic, tangentMode=RCTM_User, time=0.5, value=1}
          {tangentMode=RCTM_User, time=1, value=0}
         ]}}}
      Scale Curve=0.005
     )
    )
   ) {isSetParameters=true, module=SetVariables_#6}
   /SetVariables(
    Particles.Corona Color:LinearColor=&MakeLinearColorFromVectorAndFloat(
     "Vector (RGB)"=&MakeVectorFromLinearColorRGB(Color=<rgba 4 1.7975 0.9253 1>)
     "Float (Alpha)"=&FloatFromCurve(
      FloatCurve=DataInterface() {propertyValues {Curve {keys [
          <at 0 0 RCIM_Cubic>
          {arriveTangent=2.5, interpMode=RCIM_Cubic, leaveTangent=2.5, time=0.165, value=0.65}
          {arriveTangent=0.1786, interpMode=RCIM_Cubic, leaveTangent=0.1786, time=0.4, value=1}
          {arriveTangent=-1.6667, interpMode=RCIM_Cubic, leaveTangent=-1.6667, time=0.725, value=0.75}
          <at 1 0 RCIM_Cubic>
         ]}}}
      Scale Curve=&Multiply_Float(
       A=0.5
       B=&FloatFromCurve(FloatCurve=DataInterface() {propertyValues {Curve {keys [{interpMode=RCIM_Cubic, leaveTangent=-3.7201, tangentMode=RCTM_Break, time=0, value=1}, <at 1 0>]}}})
      )
     )
    )
    Particles.Corona Size:Vector2f=&Vector2DFromFloat(Value=&FloatFromCurve(
      FloatCurve=DataInterface() {propertyValues {Curve {keys [<at 0 1>,<at 0.5 0.7>,<at 1 1>]}}}
      Scale Curve=&Waveform(
       Global Amplitude Scale=&Multiply_Float(A=1500, B=&FloatFromCurve(FloatCurve=$System.HandchargeCurve, CurveIndex=$Emitter.NormalizedLoopAge))
       "Waveform [1]"=::NewEnumerator12
       "[1] Phase Scale"=50
       "[1] Frequency"=1
       "[1] Amplitude Min/Max"=<0.85 1>
      )
     ))
   ) {isSetParameters=true, module=SetVariables_#7}
   &DynamicMaterialParameters(
    Index 3 Param 1=&FloatFromCurve(FloatCurve=DataInterface() {propertyValues {Curve {keys [<at 0 0.7 RCIM_Cubic>,<at 0.5 0.25 RCIM_Cubic>,<at 1 0.7 RCIM_Cubic>]}}})
    Index 3 Param 2=1
    Index 3 Param 3=&RandomRangeFloat(Minimum=0.25, Maximum=0.35)
    Index 3 Param 4=1
    Module.Param0WriteEnabled3=true
    Module.Param1WriteEnabled3=true
    Module.Param2WriteEnabled3=true
    Module.Param3WriteEnabled3=true
    Write Parameter Index 0=false
    Write Parameter Index 3=true
   )
   &ScaleColor(
    Scale Alpha=&FloatFromCurve(FloatCurve=DataInterface() {propertyValues {Curve {keys [{interpMode=RCIM_Cubic, leaveTangent=-3.7201, tangentMode=RCTM_Break, time=0, value=1}, <at 1 0>]}}})
    ScaleA=true
    ScaleRGB=true
   ) {module=ScaleColor001}
   /SetVariables(Particles.Animorphic Color:LinearColor=&MultiplyLinearColor(
     Linear Color=$Particles.Initial.Animorphic Color
     Scale Factor=&MakeVector4FromVectorAndFloat(
      "Vector (RGB)"=<1 1 1>
      "Float (Alpha)"=&FloatFromCurve(FloatCurve=DataInterface() {propertyValues {Curve {keys [{interpMode=RCIM_Cubic, leaveTangent=-3.7201, tangentMode=RCTM_Break, time=0, value=1}, <at 1 0>]}}})
     )
    )) {isSetParameters=true, module=SetVariables_#8}
   /SetVariables(Particles.Position=&AddVectorToPosition(
     Position=$Particles.Initial.Position
     Vector=&TransformVector(Vector=$User.HandPosOffset, Source Space=::World, Destination Space=::Local)
    )) {isSetParameters=true, module=SetVariables_#9}
  ]
  rendererClasses [&NiagaraSpriteRendererProperties]
  renderers [
   {rendererClass=&NiagaraSpriteRendererProperties}
   {rendererClass=&NiagaraSpriteRendererProperties, rendererIndex=1}
   {rendererClass=&NiagaraSpriteRendererProperties, rendererIndex=2}
   {rendererClass=&NiagaraSpriteRendererProperties, rendererIndex=3}
   {rendererClass=&NiagaraSpriteRendererProperties, rendererIndex=4}
   {rendererClass=&NiagaraSpriteRendererProperties, rendererIndex=5}
  ]
 }
 Energy_Offset {
  emitterUpdateScript [
   &EmitterState(
    Life Cycle Mode=::NewEnumerator1
    Loop Behavior=::NewEnumerator1
    Loop Delay=$User.Energy Wind Delay
    Loop Duration=&Subtract_Float(A=$System.CurrentLoopDuration, B=$User.Energy Wind Delay)
    UseLoopDelay=true
   )
   &SpawnBurst_Instantaneous(Spawn Count=1, Use Spawn Probability=false)
  ]
  particleSpawnScript [
   &InitializeParticle(
    Lifetime=&Subtract_Float(A=$System.CurrentLoopDuration, B=$User.Energy Wind Delay)
    Mesh Scale Mode=::NewEnumerator3
    Mesh Uniform Scale=&MakeCustomFloatFromBool(
     Bool=&IsPlatformSetActive(New Platform Set=DataInterface() {propertyValues {Platforms {qualityLevelMask=-29}}})
     True Float=1
     False Float=6
    )
    UsePositionOffset=false
    Write Material Random=true
   )
   &InitialMeshOrientation(Orientation Axis=<0 0 1>, Orientation Vector=<0 0 1>, Rotation=<0 0.25 0>, Use Orientation Vector=true, Use Rotation Vector=true)
   &DynamicMaterialParameters(
    Index 0 Param 2=1
    Index 0 Param 3=1
    Index 0 Param 4=1
    Module.Index0Float0Write=true
    Module.Index0Float1Write=true
    Module.Index0Float2Write=true
    Module.Index0Float3Write=true
   )
   &InitialMeshOrientation(Orientation Axis=<0 0 1>, Orientation Vector=<0 0 1>, Rotation=<-0.8 0 0>, Use Orientation Vector=true, Use Rotation Vector=true) {module=InitialMeshOrientation001}
   &SystemLocation(MaskSpawnGroup=false, Offset=<2 -2 -2>)
  ]
  particleUpdateScript [
   &ParticleState()
   &ScaleMeshSize(Scale Factor=&VectorFromFloat(Value=&FloatFromCurve(
      FloatCurve=DataInterface() {propertyValues {Curve {keys [<at 0 2>, {interpMode=RCIM_Cubic, tangentMode=RCTM_User, time=0.9, value=1.15}, <at 1 0>]}}}
      Scale Curve=$User.FlareControl
     )))
   &DynamicMaterialParameters(
    Index 0 Param 1=&FloatFromCurve(FloatCurve=DataInterface() {propertyValues {Curve {keys [
         {arriveTangent=-0.2315, interpMode=RCIM_Cubic, leaveTangent=-0.2315, tangentMode=RCTM_User, time=0, value=0.25}
         {arriveTangent=0.25, interpMode=RCIM_Cubic, leaveTangent=0.25, time=0.95, value=0.35}
         <at 1 0.5>
        ]}}})
    Module.Param0WriteEnabled=true
    Module.Param1WriteEnabled=false
    Module.Param2WriteEnabled=false
    Module.Param3WriteEnabled=false
   ) {module=DynamicMaterialParameters001}
   &ScaleColor(
    Scale Alpha=&FloatFromCurve(
     FloatCurve=DataInterface() {propertyValues {Curve {keys [
         {interpMode=RCIM_Cubic, tangentMode=RCTM_User, time=0, value=0}
         {arriveTangent=1.4001, tangentMode=RCTM_User, time=0.2, value=1}
         <at 0.8 1>
         <at 1 0>
        ]}}}
     Scale Curve=&FloatFromCurve(FloatCurve=DataInterface() {propertyValues {Curve {keys [<at 0 0.5>,<at 0.2 1>]}}})
    )
    Scale RGB=&MakeVectorFromLinearColorRGB(Color=&ColorFromCurve(ColorCurve=DataInterface() {propertyValues {
        AlphaCurve {keys [<at 0 1 RCIM_Cubic>,<at 1 1 RCIM_Cubic>]}
        BlueCurve {keys [<at 0.0297 0 RCIM_Cubic>, {arriveTangent=1.0446, interpMode=RCIM_Cubic, leaveTangent=1.0446, time=0.8825, value=0.487}, <at 0.987 1>]}
        GreenCurve {keys [<at 0.0297 0 RCIM_Cubic>, {arriveTangent=1.0446, interpMode=RCIM_Cubic, leaveTangent=1.0446, time=0.8825, value=0.0245}, <at 0.987 1>]}
        RedCurve {keys [<at 0.0297 1 RCIM_Cubic>,<at 0.8825 0.8788 RCIM_Cubic>,<at 0.987 1>]}
       }}))
    ScaleA=true
    ScaleRGB=true
   )
   &ScaleColor(
    Scale Alpha=&FNTimeOfDayScalar(Night Scale=0.3, Morning Scale=0.3, Evening Scale=0.5)
    ScaleA=true
    ScaleRGB=true
   ) {module=ScaleColor001}
   &MeshRotationForce(Rotation=&MakeVector(
     X=&FloatFromCurve(
      FloatCurve=DataInterface() {propertyValues {Curve {keys [
          <at 0 0 RCIM_Cubic>
          {interpMode=RCIM_Cubic, tangentMode=RCTM_User, time=0.8, value=0.1}
          {arriveTangent=1.9641, tangentMode=RCTM_User, time=1, value=2}
         ]}}}
      Scale Curve=-1
     )
     Y=0
     Z=0
    ))
   &SolveForcesAndVelocity(Clamp Velocity=false, Limit Acceleration=false)
  ]
  rendererClasses [&NiagaraMeshRendererProperties]
  renderers [{rendererClass=&NiagaraMeshRendererProperties, rendererIndex=0}]
 }
 Glow001 {
  emitterUpdateScript [
   &EmitterState()
   &SpawnBurst_Instantaneous(Spawn Count=1, Use Spawn Probability=false)
  ]
  particleSpawnScript [&InitializeParticle(
    Color=<rgba 0.3073 0 0.1342 1>
    Lifetime=$System.CurrentLoopDuration
    Sprite Size Mode=::NewEnumerator3
    Uniform Sprite Size=&MakeCustomFloatFromBool(
     Bool=&IsPlatformSetActive(New Platform Set=DataInterface() {propertyValues {Platforms {qualityLevelMask=-29}}})
     True Float=600
     False Float=3000
    )
    UsePositionOffset=false
    Write Material Random=true
   )]
  particleUpdateScript [
   &ParticleState()
   &ScaleSpriteSize(Scale Factor=&Vector2DFromFloat(Value=&FloatFromCurve(
      FloatCurve=DataInterface() {propertyValues {Curve {keys [
          {arriveTangent=0.4346, interpMode=RCIM_Cubic, leaveTangent=0.4346, tangentMode=RCTM_User, time=0, value=0}
          {arriveTangent=0.0473, interpMode=RCIM_Cubic, leaveTangent=0.0473, tangentMode=RCTM_User, time=0.2, value=1}
          {arriveTangent=-0.0078, interpMode=RCIM_Cubic, leaveTangent=-0.0078, tangentMode=RCTM_User, time=0.95, value=1}
          <at 1 0 RCIM_Cubic>
         ]}}}
      Scale Curve=&RandomRangeFloat(Minimum=1, Maximum=1, Evaluation Type=::NewEnumerator1)
     )))
   &ScaleColor(
    Scale Alpha=&FloatFromCurve(
     FloatCurve=DataInterface() {propertyValues {Curve {keys [
         {arriveTangent=0.4005, interpMode=RCIM_Cubic, leaveTangent=0.4005, tangentMode=RCTM_User, time=0, value=0}
         {arriveTangent=0.0256, interpMode=RCIM_Cubic, leaveTangent=0.0256, tangentMode=RCTM_User, time=0.2, value=1}
         {arriveTangent=0.01, interpMode=RCIM_Cubic, leaveTangent=0.01, tangentMode=RCTM_User, time=0.95, value=1}
         <at 1 0 RCIM_Cubic>
        ]}}}
     Scale Curve=&Multiply_Float(A=&FNTimeOfDayScalar(Night Scale=1, Morning Scale=0.5, Evening Scale=0.5), B=$User.FlareControl)
    )
    ScaleA=true
    ScaleRGB=true
   )
   /SetVariables(
    Particles.GlowColor2:LinearColor=&MakeLinearColorFromVectorAndFloat(
     "Vector (RGB)"=&MakeVectorFromLinearColorRGB(Color=<rgba 0.4427 0.6706 5 1>)
     "Float (Alpha)"=&Multiply_Float(
      A=0.2
      B=&FloatFromCurve(
       FloatCurve=DataInterface() {propertyValues {Curve {keys [
           {arriveTangent=0.4005, interpMode=RCIM_Cubic, leaveTangent=0.4005, tangentMode=RCTM_User, time=0, value=0}
           {arriveTangent=0.0256, interpMode=RCIM_Cubic, leaveTangent=0.0256, tangentMode=RCTM_User, time=0.2, value=1}
           {arriveTangent=0.01, interpMode=RCIM_Cubic, leaveTangent=0.01, tangentMode=RCTM_User, time=0.95, value=1}
           <at 1 0 RCIM_Cubic>
          ]}}}
       Scale Curve=&FNTimeOfDayScalar(Night Scale=0.3, Morning Scale=0.2, Evening Scale=0.5)
      )
     )
    )
    Particles.GlowSize2:Vector2f=&Vector2DFromFloat(Value=&Multiply_Float(
      A=1000
      B=&FloatFromCurve(
       FloatCurve=DataInterface() {propertyValues {Curve {keys [
           {arriveTangent=0.4346, interpMode=RCIM_Cubic, leaveTangent=0.4346, tangentMode=RCTM_User, time=0, value=0}
           {arriveTangent=0.0473, interpMode=RCIM_Cubic, leaveTangent=0.0473, tangentMode=RCTM_User, time=0.2, value=1}
           {arriveTangent=-0.0078, interpMode=RCIM_Cubic, leaveTangent=-0.0078, tangentMode=RCTM_User, time=0.85, value=1}
           <at 1 0 RCIM_Cubic>
          ]}}}
       Scale Curve=&RandomRangeFloat(Minimum=0.7, Maximum=1, Evaluation Type=::NewEnumerator1)
      )
     ))
   ) {isSetParameters=true}
   /SetVariables(
    Particles.FlareColor=<rgba 1.3432 0.2124 10 1>
    Particles.Flaresize:Vector2f=&Multiply_Vector2D_ByFloat(
     Float=&Waveform("Waveform [1]"=::NewEnumerator12, "[1] Frequency"=10, "[1] Amplitude Min/Max"=<0.8 1>)
     Vector2D=&Multiply_Vector2D_ByFloat(
      Float=&FloatFromCurve(
       FloatCurve=DataInterface() {propertyValues {Curve {
          defaultValue=1
          keys [
           {interpMode=RCIM_Cubic, tangentMode=RCTM_User, time=0, value=0}
           {interpMode=RCIM_Cubic, tangentMode=RCTM_User, time=0.3, value=1}
           {arriveTangent=-1.8182, interpMode=RCIM_Cubic, leaveTangent=-1.8182, time=0.65, value=1}
           <at 0.85 0 RCIM_Cubic>
          ]
         }}}
       Scale Curve=2000
      )
      Vector2D=<0.65 1.5>
     )
    )
   ) {isSetParameters=true}
  ]
  rendererClasses [&NiagaraSpriteRendererProperties]
  renderers [
   {rendererClass=&NiagaraSpriteRendererProperties}
   {rendererClass=&NiagaraSpriteRendererProperties, rendererIndex=1}
   {rendererClass=&NiagaraSpriteRendererProperties, rendererIndex=2}
  ]
 }
 LeakParticle {
  simTarget=GPUComputeSim
  emitterSpawnScript [&self:InitialRandomArray(FA:NiagaraDataInterfaceArrayFloat=$User.RandomArray) {module=InitialRandomArray}]
  emitterUpdateScript [
   &EmitterState()
   &SpawnBurst_Instantaneous(Spawn Count=83, Use Spawn Probability=false, UseLoopCountLimit=false) {enabled=false}
   &SpawnRate(
    Spawn Probability=$System.NormalizedLoopAge
    SpawnRate=&FloatFromCurve(
     FloatCurve=DataInterface() {propertyValues {Curve {keys [<at 0 0>,<at 0.85 1>,<at 0.95 0>]}}}
     CurveIndex=$Emitter.NormalizedLoopAge
     Scale Curve=&Multiply_Float(A=250, B=$User.FlareControl)
    )
    Use Spawn Probability=true
   ) {module=SpawnRate001}
  ]
  particleSpawnScript [
   &InitializeParticle(
    Color=<rgba 2 0.2 0.35 1>
    Lifetime Max=0.6585
    Lifetime Min=0.5384
    Lifetime Mode=::NewEnumerator1
    Sprite Size=&Multiply_Vector2D_ByFloat(Float=&RandomRangeFloat(Minimum=2, Maximum=5), Vector2D=<1 3>)
    Sprite Size Mode=::NewEnumerator4
    UsePositionOffset=false
    Write Material Random=true
   )
   &ShapeLocation(
    Additional Quaternion Rotation=false
    MaskSpawnGroup=false
    Override Seed=false
    Rotation Mode=::NewEnumerator3
    Rotation Quaternion=&FindQuatBetweenTwoVectors(
     From Vector=<0 0 1>
     To Vector=&RotateVectorByQuaternion(Vector To Rotate=<0 1 0>, Quaternion=$Particles.SkeletalMeshLocation001.SampledRotation)
    )
    Sphere Radius=2000
    UseOwnerScale=true
   )
   &ShapeLocation(MaskSpawnGroup=false, Override Seed=false, Sphere Radius=2.502, UseOwnerScale=true) {enabled=false, module=ShapeLocation001}
   &CameraOffset(Camera Offset Amount=-10)
   &InitialMeshOrientation(
    Mesh Orientation Mode=::NewEnumerator6
    Orientation Quaternion:Quat4f=&FindQuatBetweenTwoVectors(
     From Vector=<0 0 1>
     To Vector=&RotateVectorByQuaternion(Vector To Rotate=<0 1 0>, Quaternion=$Particles.SkeletalMeshLocation001.SampledRotation)
    )
    Use Rotation Vector=false
   )
   &AddVelocity(
    Module.Use Falloff Distance=false
    Override Seed=false
    Velocity Mode=::NewEnumerator1
    Velocity Speed=&Multiply_Float(
     A=&Multiply_Float(A=38.0946, B=&RandomRangeFloat(Minimum=100, Maximum=200))
     B=$System.NormalizedLoopAge
    )
   )
  ]
  particleUpdateScript [
   &ParticleState()
   &ScaleSpriteSize(
    Scale Factor=&MakeVector2D(
     X=1
     Y=&FloatFromCurve(FloatCurve=DataInterface() {propertyValues {Curve {keys [{interpMode=RCIM_Cubic, leaveTangent=-10.3145, tangentMode=RCTM_Break, time=0, value=4}, <at 1 1>]}}})
    )
    Scale Sprite Size Mode=::NewEnumerator2
   ) {module=ScaleSpriteSize001}
   &CurlNoiseForce(Noise Frequency=15, Noise Strength=10000, Pan Noise=false)
   &PointAttractionForce(Attraction Radius=2500, Attraction Strength=800, Kill Within Radius=false, Module.Use Falloff=true, UseAttractorOffset=false) {module=PointAttractionForce001}
   &ScaleSpriteSize(Scale Sprite Size Mode=::NewEnumerator0, Uniform Scale Factor=2)
   &Drag(Drag=&RandomRangeFloat(Minimum=8, Maximum=25), Use Linear Drag=true, Use Rotational Drag=false)
   &SolveForcesAndVelocity(Clamp Velocity=false, Limit Acceleration=false)
   &ScaleSpriteSizeBySpeed(Max Scale Factor=<0.1 5>, Velocity Threshold=5633.8472)
   &ScaleColor(
    Scale Alpha=&FloatFromCurve(FloatCurve=DataInterface() {propertyValues {Curve {keys [
         {tangentMode=RCTM_User, time=0, value=1}
         {arriveTangent=-3.3056, interpMode=RCIM_Cubic, leaveTangent=-3.3056, tangentMode=RCTM_User, time=0.2, value=1}
         {arriveTangent=-0.0264, interpMode=RCIM_Cubic, leaveTangent=-0.0264, tangentMode=RCTM_User, time=1, value=0}
        ]}}})
    Scale RGB=&VectorFromFloat(Value=10)
    ScaleA=true
    ScaleRGB=true
   )
   &FN_Sprite_Attributes(
    Color=&MultiplyLinearColor(
     Linear Color=<rgba 0.151 0.003 0.0154 1>
     Scale Factor=&MakeVector4(
      X=1
      Y=1
      Z=1
      W=&FloatFromCurve(FloatCurve=DataInterface() {propertyValues {Curve {keys [
           {arriveTangent=0.1913, interpMode=RCIM_Cubic, leaveTangent=0.1913, tangentMode=RCTM_User, time=0, value=0}
           <at 0.5 1 RCIM_Cubic>
           {arriveTangent=-0.4533, interpMode=RCIM_Cubic, leaveTangent=-0.4533, tangentMode=RCTM_User, time=1, value=0}
          ]}}})
     )
    )
    Create Color Attribute=true
    Create SpriteSize Attribute=true
    SpriteSize=<179.3112 373.3694>
    Write to Module Namespace=true
   )
  ]
  rendererClasses [&NiagaraSpriteRendererProperties]
  renderers [
   {rendererClass=&NiagaraSpriteRendererProperties}
   {rendererClass=&NiagaraSpriteRendererProperties, rendererIndex=1}
  ]
 }
 LongFlare {
  bEnabled=false
  emitterUpdateScript [
   &EmitterState(
    Life Cycle Mode=::NewEnumerator1
    Loop Behavior=::NewEnumerator1
    Loop Delay=0.2
    Loop Duration=$System.CurrentLoopDuration
    UseLoopDelay=true
   )
   &SpawnBurst_Instantaneous(Spawn Count=1, Use Spawn Probability=false)
  ]
  particleSpawnScript [
   /Niagara/Modules/Spawn/Initialization/InitializeParticle(
    Color=<rgba 3.2135 3.852 10 1>
    Lifetime=$User.Duration
    Module.Write Color=true
    Module.Write Lifetime=true
    Module.Write Mass=true
    Module.Write Material Random=true
    Module.Write Position=false
    Module.Write Scale=false
    Module.Write SpriteRotation=true
    Module.Write SpriteSize=true
    Position=&SimulationPosition()
    Sprite Size=&Multiply_Vector2D_ByFloat(
     Float=$User.ElementSizeScale
     Vector2D=&Multiply_Vector2D_ByFloat(Float=0.136, Vector2D=<11981.0986 4753.9893>)
    )
   )
   &DynamicMaterialParameters(
    Index 0 Param 1=0.2
    Index 0 Param 2=0.1
    Module.Param0WriteEnabled=true
    Module.Param0WriteEnabled1=true
    Module.Param0WriteEnabled2=true
    Module.Param0WriteEnabled3=true
    Module.Param1WriteEnabled=true
    Module.Param1WriteEnabled1=true
    Module.Param1WriteEnabled2=true
    Module.Param1WriteEnabled3=true
    Module.Param2WriteEnabled=true
    Module.Param2WriteEnabled1=true
    Module.Param2WriteEnabled2=true
    Module.Param2WriteEnabled3=true
    Module.Param3WriteEnabled=true
    Module.Param3WriteEnabled1=true
    Module.Param3WriteEnabled2=true
    Module.Param3WriteEnabled3=true
    Write Parameter Index 1=true
    Write Parameter Index 2=true
    Write Parameter Index 3=true
   )
   /SetVariables(
    Particles.FlickFlareColor:LinearColor=<rgba 1.6659 0.2768 12 1>
    Particles.FlickFlareSize:Vector2f=&Multiply_Vector2D_ByFloat(Float=$User.ElementSizeScale, Vector2D=&Multiply_Vector2D_ByFloat(Float=0.15, Vector2D=<5926.7925 5320.0742>))
   ) {isSetParameters=true}
  ]
  particleUpdateScript [
   &ParticleState()
   &SolveForcesAndVelocity(Clamp Velocity=false, Limit Acceleration=false)
   &ScaleColor(
    Scale Alpha=&FloatFromCurve(FloatCurve=DataInterface() {propertyValues {Curve {keys [
         {interpMode=RCIM_Cubic, leaveTangent=0.1827, tangentMode=RCTM_Break, time=0.3, value=0}
         {interpMode=RCIM_Cubic, leaveTangent=2.9317, tangentMode=RCTM_Break, time=0.9, value=1}
         <at 1 0 RCIM_Cubic>
        ]}}})
    ScaleA=true
    ScaleRGB=true
   )
   &ScaleSpriteSize(Scale Factor=&MakeVector2D(
     X=&FloatFromCurve(
      FloatCurve=DataInterface() {propertyValues {Curve {keys [
          {arriveTangent=4.441, interpMode=RCIM_Cubic, leaveTangent=4.441, tangentMode=RCTM_User, time=0, value=0}
          {arriveTangent=-0.0405, interpMode=RCIM_Cubic, leaveTangent=-0.0405, tangentMode=RCTM_User, time=1, value=2}
         ]}}}
      Scale Curve=$User.FlareControl
     )
     Y=1
    ))
   &ScaleSpriteSize(
    Scale Factor=&MakeVector2D(X=&Waveform("Waveform [1]"=::NewEnumerator0, "[1] Frequency"=10, "[1] Amplitude Min/Max"=<0.5 1>), Y=1)
    Scale Sprite Size Mode=::NewEnumerator2
   ) {module=ScaleSpriteSize001}
   /SetVariables(
    Particles.FlickFlareColor:LinearColor=&MultiplyLinearColor(
     Linear Color=$Particles.Initial.FlickFlareColor
     Scale Factor=&MakeVector4(
      X=1
      Y=1
      Z=1
      W=&FloatFromCurve(FloatCurve=DataInterface() {propertyValues {Curve {
          defaultValue=1
          keys [{interpMode=RCIM_Cubic, tangentMode=RCTM_User, time=0, value=0}, {interpMode=RCIM_Cubic, tangentMode=RCTM_User, time=1, value=1}]
         }}})
     )
    )
    Particles.FlickFlareOffset:NiagaraFloat=&Multiply_Float(
     A=300
     B=&FloatFromCurve(
      FloatCurve=DataInterface() {propertyValues {Curve {keys [<at 0 0>,<at 1 1>]}}}
      CurveIndex=/Game/Effects/Niagara/Modules/Math/FN_CameraDistance(
       0Dist:NiagaraFloat=300
       1Dist:NiagaraFloat=2000
       Camera Query:NiagaraDataInterfaceCamera=DataInterface()
       Position:NiagaraPosition=$Particles.Position
      )
     )
    )
    Particles.FlickFlareSize:Vector2f=&Multiply_Vector2D(
     A=$Particles.Initial.FlickFlareSize
     B=&Vector2DFromFloat(Value=&FloatFromCurve(
       FloatCurve=DataInterface() {propertyValues {Curve {keys [
           {arriveTangent=4.1785, interpMode=RCIM_Cubic, leaveTangent=4.1785, tangentMode=RCTM_User, time=0, value=0}
           {arriveTangent=1.9091, interpMode=RCIM_Cubic, leaveTangent=1.9091, tangentMode=RCTM_User, time=0.2, value=1}
           {arriveTangent=-0.0405, interpMode=RCIM_Cubic, leaveTangent=-0.0405, tangentMode=RCTM_User, time=0.98, value=2}
           <at 1 0 RCIM_Cubic>
          ]}}}
       Scale Curve=&Waveform("Waveform [1]"=::NewEnumerator0, "[1] Frequency"=7, "[1] Amplitude Min/Max"=<0.7 1>)
      ))
    )
   ) {isSetParameters=true}
  ]
  rendererClasses [&NiagaraSpriteRendererProperties]
  renderers [
   {rendererClass=&NiagaraSpriteRendererProperties}
   {rendererClass=&NiagaraSpriteRendererProperties, rendererIndex=1}
  ]
 }
 Minimal {
  emitterUpdateScript [
   &EmitterState(
    Life Cycle Mode=::NewEnumerator1
    Loop Behavior=::NewEnumerator1
    Loop Delay=$System.HandchargeStartTime
    Loop Duration=$System.HandchargeLoopTime
    UseLoopDelay=true
   )
   &SpawnBurst_Instantaneous(Use Spawn Probability=false, UseLoopCountLimit=false)
   /SetVariables(Emitter.ChargingProgressSquared:NiagaraFloat=&FloatFromCurve(FloatCurve=$System.HandchargeCurve, CurveIndex=$Emitter.NormalizedLoopAge)) {isSetParameters=true, module=SetVariables_#10}
  ]
  particleSpawnScript [/Niagara/Modules/Spawn/Initialization/V2/InitializeParticle(Lifetime=4, Position Offset=<0 50 150>, Position Offset Coordinate Space=::World, UsePositionOffset=true, Write Material Random=true)]
  particleUpdateScript [
   &ParticleState()
   /SetVariables(Particles.Position=&AddVectorToPosition(
     Position=$Particles.Initial.Position
     Vector=&TransformVector(Vector=$User.HandPosOffset, Source Space=::World, Destination Space=::Local)
    )) {isSetParameters=true, module=SetVariables_#9}
   &FN_Light_Attributes(
    Color=<rgba 35 0 16.9788 1>
    Create Color Attribute=true
    Create Exponent Attribute=true
    Create Radius Attribute=true
    Exponent=2
    Radius=1200
    Scale Color=&Waveform(
     Global Amplitude Scale=&Lerp_Float(A=1, B=30, Alpha=$Emitter.ChargingProgressSquared)
     "Waveform [1]"=::NewEnumerator12
     "[1] Frequency"=500
     "[1] Amplitude Min/Max"=<0.4 1>
    )
   )
  ]
  rendererClasses [&NiagaraLightRendererProperties]
  renderers [{rendererClass=&NiagaraLightRendererProperties, rendererIndex=0}]
 }
 PostProcess {
  emitterUpdateScript [
   &EmitterState()
   &SpawnBurst_Instantaneous(Spawn Time=0.05, Use Spawn Probability=false, UseLoopCountLimit=false)
  ]
  particleSpawnScript [&InitializeParticle(
    Lifetime=$System.CurrentLoopDuration
    Mesh Scale Mode=::NewEnumerator3
    Sprite Size=&Multiply_Vector2D_ByFloat(Float=&Multiply_Float(A=$User.ElementSizeScale, B=$User.PostSizeScale), Vector2D=<3948.9736 3201.8015>)
    Sprite Size Mode=::NewEnumerator4
    UsePositionOffset=false
    Write Material Random=true
   )]
  particleUpdateScript [
   &ParticleState()
   /SetVariables(Particles.BlendTest:NiagaraFloat=&FloatFromCurve(
     FloatCurve=DataInterface() {propertyValues {Curve {keys [
         {arriveTangent=0.421, interpMode=RCIM_Cubic, leaveTangent=0.421, tangentMode=RCTM_User, time=0, value=0}
         {arriveTangent=-0.0017, interpMode=RCIM_Cubic, leaveTangent=-0.0017, tangentMode=RCTM_User, time=5.3, value=0.7}
        ]}}}
     CurveIndex=$Emitter.Age
     Scale Curve=&Subtract_Float(
      A=1
      B=&FN_LOD_Distance_Remap(Start Distance=4000, End Distance=6000, Minimum Output=0, Maximum Output=1, Clamp Output=true)
     )
    )) {isSetParameters=true}
   /SetVariables(Particles.ComponentsEnabled:NiagaraBool=&Boolean_NotOperation(Bool=&Boolean_OrOperation(
      Bool1=&SetBoolByFloatComparison(
       A=&FN_LOD_Distance_Remap(Start Distance=4000, End Distance=6000, Minimum Output=0, Maximum Output=1, Clamp Output=true)
       B=0.99
      )
      Bool2=&SetBoolByFloatComparison(A=$Emitter.Age, B=$System.CurrentLoopDuration)
     ))) {isSetParameters=true}
  ]
  rendererClasses [&NiagaraComponentRendererProperties]
  renderers [{rendererClass=&NiagaraComponentRendererProperties, rendererIndex=0}]
 }
 PostRadialBlur {
  bEnabled=false
  emitterUpdateScript [
   &EmitterState(
    Life Cycle Mode=::NewEnumerator1
    Loop Behavior=::NewEnumerator1
    Loop Delay=$System.HandchargeStartTime
    Loop Duration=$System.HandchargeLoopTime
    UseLoopDelay=true
   )
   &SpawnBurst_Instantaneous(Use Spawn Probability=false, UseLoopCountLimit=true)
   /SetVariables(Emitter.ChargingProgressSquared:NiagaraFloat=&FloatFromCurve(FloatCurve=$System.HandchargeCurve, CurveIndex=$Emitter.NormalizedLoopAge)) {isSetParameters=true, module=SetVariables_#10}
  ]
  particleSpawnScript [&InitializeParticle(
    Color=<rgba 1.8427 0.4465 2 1>
    Lifetime=3
    Mesh Scale Mode=::NewEnumerator3
    Position Offset=<0 50 150>
    Position Offset Coordinate Space=::World
    Sprite Size Mode=::NewEnumerator3
    Uniform Sprite Size=1000
    UsePositionOffset=true
    Write Material Random=true
   )]
  particleUpdateScript [
   &ParticleState()
   &ScaleSpriteSize(
    Uniform Curve Scale=$Emitter.ChargingProgressSquared
    Uniform Curve Sprite Scale=DataInterface() {propertyValues {Curve {
       defaultValue=1
       keys [{interpMode=RCIM_Cubic, tangentMode=RCTM_User, time=0, value=0}, {interpMode=RCIM_Cubic, tangentMode=RCTM_User, time=1, value=1}]
      }}}
   )
   /SetVariables(Particles.float:NiagaraFloat=&FloatFromCurve(
     FloatCurve=DataInterface() {propertyValues {Curve {keys [
         {arriveTangent=18.957, interpMode=RCIM_Cubic, leaveTangent=18.957, tangentMode=RCTM_User, time=0, value=0}
         {arriveTangent=-0.1384, interpMode=RCIM_Cubic, leaveTangent=-0.1384, tangentMode=RCTM_User, time=0.05, value=0.9847}
         {interpMode=RCIM_Cubic, leaveTangent=-3.562, time=1, value=1}
        ]}}}
     Scale Curve=&Waveform(Global Amplitude Scale=1.5, "Waveform [1]"=::NewEnumerator12, "[1] Frequency"=20, "[1] Amplitude Min/Max"=<0.5 1>)
    )) {isSetParameters=true}
   &DynamicMaterialParameters(
    Index 0 Param 1=$Particles.float
    Index 0 Param 2=$Particles.float
    Index 0 Param 3=0.8
    Index 1 Param 1=0.2
    Index 1 Param 2=500
    Index 1 Param 3=0.2
    Index 2 Param 1=$Particles.Timeline.Time
    Index 2 Param 2=$Particles.Timeline.Time
    Index 2 Param 3=$Particles.Timeline.Time
    Index 2 Param 4=0.5
    Index 3 Param 1=2
    Index 3 Param 2=200
    Module.Index0Param0Write=true
    Module.Index0Param1Write=true
    Module.Index0Param2Write=true
    Module.Index0Param3Write=true
    Module.Index1Param0Write=true
    Module.Index1Param1Write=true
    Module.Index1Param2Write=true
    Module.Index1Param3Write=true
    Module.Index2Param0Write=true
    Module.Index2Param1Write=true
    Module.Index2Param2Write=true
    Module.Index2Param3Write=true
    Module.Index3Param0Write=true
    Module.Index3Param1Write=true
    Module.Index3Param2Write=true
    Module.Index3Param3Write=true
    Write Parameter Index 1=true
    Write Parameter Index 2=true
    Write Parameter Index 3=true
   )
   /SetVariables(Particles.Position=&AddVectorToPosition(
     Position=$Particles.Initial.Position
     Vector=&TransformVector(Vector=$User.HandPosOffset, Source Space=::World, Destination Space=::Local)
    )) {isSetParameters=true, module=SetVariables_#9}
   &ScaleColor(
    Scale RGB=&Multiply_VectorByFloat(
     Vector=&VectorFromFloat(Value=&Waveform("Waveform [1]"=::NewEnumerator12, "[1] Frequency"=1, "[1] Amplitude Min/Max"=<1 1>))
     Float=&FloatFromCurve(
      FloatCurve=DataInterface() {propertyValues {Curve {
         defaultValue=1
         keys [
          {arriveTangent=4.1785, interpMode=RCIM_Cubic, leaveTangent=4.1785, tangentMode=RCTM_User, time=0, value=0}
          {arriveTangent=1.1179, interpMode=RCIM_Cubic, leaveTangent=1.1179, tangentMode=RCTM_User, time=0.2042, value=0.6228}
          {arriveTangent=0.0224, interpMode=RCIM_Cubic, leaveTangent=0.0224, tangentMode=RCTM_User, time=1, value=1}
         ]
        }}}
      CurveIndex=$Emitter.ChargingProgressSquared
     )
    )
    ScaleA=true
    ScaleRGB=true
   )
  ]
  rendererClasses [&NiagaraSpriteRendererProperties]
  renderers [{rendererClass=&NiagaraSpriteRendererProperties}]
 }
}
summary {
 systemName=NS_Pistol_Canary_MM_Charge
 emitters [
  {emitterName=LeakParticle, simTarget=GPUComputeSim, rendererClasses [&NiagaraSpriteRendererProperties]}
  {bEnabled=false, emitterName=LongFlare, rendererClasses [&NiagaraSpriteRendererProperties]}
  {emitterName=PostProcess, rendererClasses [&NiagaraComponentRendererProperties]}
  {emitterName=Air_Tendrils_Large, rendererClasses [&NiagaraMeshRendererProperties]}
  {emitterName=Energy_Offset, rendererClasses [&NiagaraMeshRendererProperties]}
  {emitterName=Glow001, rendererClasses [&NiagaraSpriteRendererProperties]}
  {emitterName=Burst, simTarget=GPUComputeSim, rendererClasses [&NiagaraSpriteRendererProperties]}
  {emitterName=BurstSparks, simTarget=GPUComputeSim, rendererClasses [&NiagaraSpriteRendererProperties]}
  {bEnabled=false, emitterName=PostRadialBlur, rendererClasses [&NiagaraSpriteRendererProperties]}
  {emitterName=Minimal, rendererClasses [&NiagaraLightRendererProperties]}
  {emitterName=EnergyBall, rendererClasses [&NiagaraSpriteRendererProperties]}
 ]
 userVariables [
  User.BeamEnd_0 {type=NiagaraPosition, description="", defaultValue {}}
  User.ChargingProgress {type=NiagaraFloat, defaultValue=1, description=""}
  User.Duration {type=NiagaraFloat, defaultValue=30, description=""}
  User.ElementSizeScale {type=NiagaraFloat, defaultValue=0, description=""}
  User.Energy Wind Delay {type=NiagaraFloat, defaultValue=2, description=""}
  User.FlareControl {type=NiagaraFloat, defaultValue=1, description=""}
  User.HandPosOffset {type=Vector3f, description="", defaultValue {struct=Vector3f, value=<0 0 0>}}
  User.PostSizeScale {type=NiagaraFloat, defaultValue=0, description=""}
  User.PrimaryColor {type=LinearColor, description="", defaultValue {struct=LinearColor, value=<rgba 3 0 0.2028 1>}}
  User.RandomArray {
   type=NiagaraDataInterfaceArrayFloat
   defaultValue {
    struct=NiagaraExt_VariableValue_DataInterface
    value {
     dataInterface=&self:NiagaraDataInterfaceArrayFloat_1
     dataInterfaceClass=&NiagaraDataInterfaceArrayFloat
    }
   }
  }
 ]
}
